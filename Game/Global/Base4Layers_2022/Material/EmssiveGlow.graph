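# Unlit material for an emissive glow, colored by a blackbody Temperature and scaled by a Brightness parameter.
# The glow is weighted by Fortnite's active time of day from a material parameter collection, via a dot product.
# World height relative to the actor pivot, plus an offset, saturates into a mask that cuts glow below that level.
# Rounded V coordinate splits the UVs into full-strength and LowerBrightness halves.
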Material /Game/Global/Base4Layers_2022/Material/EmssiveGlow
graph {
 properties {pixelDepthOffsetMode=PDOM_Legacy, shadingModel=MSM_Unlit, usedShadingModels=Unlit}
 nodes [
  n1 = TextureCoordinate()
  Add(A=@n1) {constB=0.5}
  n2 = WorldPosition()
  n3 = ActorPositionWS()
  n4 = Add(A=ScalarParameter() {defaultValue=100}, B=ComponentMask(None=Subtract(A=@n2.XYZ, B=@n3)) {b=true})
  n5 = Multiply(
   A=BlackBody!(Temp=ScalarParameter() {defaultValue=2000, parameterName=Temperature})
   B=ScalarParameter() {defaultValue=25, parameterName=Brightness}
  )
  n6 = CollectionParameter() {
   collection=&FortniteMaterialParameters
   parameterName=FortniteActiveTimeOfDay
  }
  n7 = DotProduct(A=Constant4Vector() {constant=<rgba 0.2 0 0.2 1>}.RGBA, B=@n6)
  n8 = Saturate(None=ComponentMask(None=@n1) {g=true})
  Multiply(
   A=Multiply(A=Saturate(None=@n4), B=Multiply(A=@n5, B=@n7))
   B=LinearInterpolate(B=ScalarParameter() {parameterName=LowerBrightness}, Alpha=Round(None=@n8)) {constA=1}
  )
  TextureCoordinate()
 ]
}
