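# M_GlassTest is a two-sided test material, named for glass, that uses the Single Layer Water shading model.
# Scattering2 color, scaled by its alpha over 100, feeds scattering, while Absorption color multiplied by itself feeds absorption, ignoring its alpha.
# Saturated TwoSidedSign zeroes both coefficients on back faces, applying them only to front faces.
# Several loose constants and parameters, including Surface Opacity, connect to nothing shown.
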Material /Game/Global/Base4Layers_2022/Material/M_GlassTest
graph {
 properties {
  bUsedWithNanite=true
  pixelDepthOffsetMode=PDOM_Legacy
  refractionMethod=RM_IndexOfRefraction
  shadingModel=MSM_SingleLayerWater
  twoSided=true
  usedShadingModels=SingleLayerWater
 }
 nodes [
  Constant3Vector() {constant=<rgba 0 0 1 0>}
  Constant()
  Constant() {r=1}
  ScalarParameter() {defaultValue=1}
  ScalarParameter() {group=Water Shading Model, parameterName=Surface Opacity}
  n1 = VectorParameter() {group=Water Shading Model, parameterName=Scattering2, defaultValue=<rgba 0.0009 0.138 0.1649 0.02>}
  n2 = Divide(A=@n1.A) {constB=100}
  n3 = Saturate(None=TwoSidedSign())
  n4 = VectorParameter() {group=Water Shading Model, parameterName=Absorption, defaultValue=<rgba 1 0.4339 0.634 0.001>}
  SingleLayerWaterMaterialOutput!(
   ScatteringCoefficients=LinearInterpolate(A=Constant(), B=Multiply(A=@n1.RGB, B=@n2), Alpha=@n3)
   AbsorptionCoefficients=LinearInterpolate(A=Constant(), B=Multiply(A=@n4.RGB, B=@n4.RGB), Alpha=@n3)
   PhaseG
   ColorScaleBehindWater
  )
  VectorParameter() {defaultValue=<rgba 0.3042 0.8455 0.6889 1>}
 ]
}
