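# Layered rock and cliff material function for Fortnite environment meshes, built mostly from static switches.
# L1 rock uses UV, triplanar or Z-projected sampling, then takes sedimentary cliff, detail or baked-prop L2 layers.
# L3 adds custom or landscape-matched biome tops, then RVT blending, optional waterline, shared WPO, damage and effect functions finish it.
# Shading-path, quality and Nanite switches simplify mobile and lower-quality paths and adjust normal strength.
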MaterialFunction /Game/Global/Base4Layers_2022/MaterialFunctions/MF_Rocks_2023
graph {
 properties {description=Checker}
 nodes [
  n1 = StaticSwitchParameter(
   True=NamedRerouteUsage()
   False=TextureSample(UVs=NamedRerouteUsage()) {texture=/Game/FN_PublicTurntable/Turntable_Lowpoly/Content/BlueprintAssets/Checker}.RGB
  ) {group="A_L1_Rock Surface (TD slot 1)", parameterName=L1 is WS, sortPriority=-1}
  n2 = MaterialFunctionCall!(
   Sediment MA=NamedRerouteUsage()
   IsLandscape=StaticBool()
   Sediment Normal Intensity=ScalarParameter() {defaultValue=1, group=Cliff, parameterName=Sediment_b | L2 Cliff Normal Intensity, sortPriority=307}
   Cliffs MA=NamedRerouteUsage()
   Sediment Wave Distiortion=NamedRerouteUsage()
   Sediment Normal B Mask=NamedRerouteUsage()
   Biom Mask=NamedRerouteUsage()
  ) {materialFunction=/Game/Athena/Helios/Environments/Landscape/Materials/MaterialFunctions_Utilities/MF_CliffSediment_Blend}
  n3 = NamedRerouteUsage()
  n4 = NamedRerouteUsage()
  n5 = MaterialFunctionCall!(
   "Normal (WS)"
   Texture Scale=ScalarParameter() {defaultValue=2000, group="A_L1_Rock Surface (TD slot 1)", parameterName=L1 WS_Tile, sortPriority=0}
   Triplanar Fade=ScalarParameter() {defaultValue=10, group="A_L1_Rock Surface (TD slot 1)", parameterName=L1 WS_Triplanar Fade, sortPriority=0}
   Derivatives Scale
  ) {materialFunction=/Game/Athena/Environments/MaterialFunctions/MF_TriplanarDithered}
  n6 = TextureSampleParameter2D!(UVs=@n5.UVs, "DDX(UVs)"=@n5.DDX, "DDY(UVs)"=@n5.DDY, Apply View MipBias) {
   group="A_L1_Rock Surface (TD slot 1)"
   mipValueMode=TMVM_Derivative
   parameterName=Diffuse
   samplerSource=SSM_Wrap_WorldGroupSettings
   sortPriority=100
   texture=/Game/Environments/Asteria/Terrain/Textures/T_Asteria_RockCliff_04_D
  }
  n7 = TextureSampleParameter2D(UVs=@n5.UVs, "DDX(UVs)"=@n5.DDX, "DDY(UVs)"=@n5.DDY) {
   group="A_L1_Rock Surface (TD slot 1)"
   mipValueMode=TMVM_Derivative
   parameterName=SpecularMasks
   samplerSource=SSM_Wrap_WorldGroupSettings
   samplerType=SAMPLERTYPE_Masks
   sortPriority=105
   texture=/Game/Environments/Asteria/Terrain/Textures/T_Asteria_RockCliff_04_S
  }
  n8 = TextureSampleParameter2D(UVs=@n5.UVs, "DDX(UVs)"=@n5.DDX, "DDY(UVs)"=@n5.DDY) {
   group="A_L1_Rock Surface (TD slot 1)"
   mipValueMode=TMVM_Derivative
   parameterName=Normals
   samplerSource=SSM_Wrap_WorldGroupSettings
   samplerType=SAMPLERTYPE_Normal
   sortPriority=110
   texture=/Game/Environments/Asteria/Terrain/Textures/T_Asteria_RockCliff_04_N
  }
  n9 = SetMaterialAttributes!(
   MaterialAttributes
   BaseColor=@n6.RGB
   Opacity=@n6.RGB
   Roughness=@n7.B
   Specular=@n7.B
   Metallic=@n7.B
   Normal=Transform(None=MaterialFunctionCall!(Tangent Normals=@n8.RGB, Vertex Normals, Triplanar Mask=@n5.Mask) {materialFunction=/Game/Athena/Environments/MaterialFunctions/MF_TriplanarNormalTangentToWorld}.World Normals) {transformSourceType=TRANSFORMSOURCE_World, transformType=TRANSFORM_Tangent}
  ) {attributeSetTypes [
    #1
    #2
    #3
    #4
    #5
    #6
   ]}
  n10 = StaticBoolParameter() {defaultValue=true, desc="On: L1 Use UV0\r\nOff: L1 Use UV1", group=0_Rock/Cliff Main Settings, parameterName=L1 Use Displacement}
  n11 = StaticSwitch(
   True=TextureCoordinate()
   False=TextureCoordinate() {coordinateIndex=1}
   Value=NamedRerouteDeclaration(None=@n10) {name=L1 Displacement, nodeColor=<rgba 0 1 0.7188 1>}
  )
  n12 = StaticSwitch(
   True=@n11
   False=Multiply(A=@n11, B=ScalarParameter() {defaultValue=1, group="A_L1_Rock Surface (TD slot 1)", parameterName=L1 Tile, sortPriority=-1})
   Value=NamedRerouteDeclaration(None=StaticBoolParameter() {group=0_Rock/Cliff Main Settings, parameterName=Baked Prop UV2, sortPriority=0}) {name=IsBaked, nodeColor=<rgba 1 0.75 0 1>}
  )
  n13 = TextureSampleParameter2D!(UVs=NamedRerouteDeclaration(None=@n12) {name=L1 UV Baked, nodeColor=<rgba 0 1 0.8594 1>}, Apply View MipBias) {
   group="A_L1_Rock Surface (TD slot 1)"
   parameterName=Diffuse
   samplerSource=SSM_Wrap_WorldGroupSettings
   sortPriority=100
   texture=&T_Asteria_RockCliff_04_D
  }
  n14 = TextureSampleParameter2D(UVs=@n12) {
   group="A_L1_Rock Surface (TD slot 1)"
   parameterName=SpecularMasks
   samplerSource=SSM_Wrap_WorldGroupSettings
   samplerType=SAMPLERTYPE_Masks
   sortPriority=105
   texture=&T_Asteria_RockCliff_04_S
  }
  n15 = TextureSampleParameter2D(UVs=@n12) {
   group="A_L1_Rock Surface (TD slot 1)"
   parameterName=Normals
   samplerSource=SSM_Wrap_WorldGroupSettings
   samplerType=SAMPLERTYPE_Normal
   sortPriority=110
   texture=&T_Asteria_RockCliff_04_N
  }
  n16 = SetMaterialAttributes(BaseColor=@n13.RGB, Opacity=@n13.RGB, Roughness=@n14.B, Specular=@n14.B, Metallic=@n14.B, Normal=@n15.RGB) {attributeSetTypes [
    #1
    #2
    #3
    #4
    #5
    #6
   ]}
  n17 = GetMaterialAttributes(None=StaticSwitchParameter(True=@n9, False=@n16) {group="A_L1_Rock Surface (TD slot 1)", parameterName=L1 is WS, sortPriority=-1}) {attributeGetTypes [#6]}
  n18 = ComponentMask(None=@n17.Normal) {g=true, r=true}
  n19 = ScalarParameter() {defaultValue=1.2, group="A_L1_Rock Surface (TD slot 1)", parameterName="L1 Normals Value Baked Prop (non-Nanite)", sortPriority=21}
  n20 = ScalarParameter() {defaultValue=2, group="A_L1_Rock Surface (TD slot 1)", parameterName="L1 Normals Value (non-Nanite)", sortPriority=21}
  n21 = NamedRerouteUsage()
  n22 = ScalarParameter() {defaultValue=1, group="A_L1_Rock Surface (TD slot 1)", parameterName="L1 Normals Value (Nanite)", sortPriority=21}
  n23 = NaniteReplace!(Default=Multiply(A=StaticSwitch(True=@n19, False=@n20, Value=@n21), B=@n22) {constA=2}, Nanite=@n22)
  n24 = Normalize(VectorInput=AppendVector(A=Multiply(A=@n18, B=@n23), B=ComponentMask(None=@n17.Normal) {b=true}))
  n25 = StaticSwitch(
   True=ShadingPathSwitch(
    Default=@n2.MA
    Mobile=MaterialFunctionCall!(Baked Material=@n3, Sedimentary Material=@n4) {materialFunction=/Game/Athena/Helios/Environments/Landscape/Materials/QualityLevelLayers/MF_Helios_CliffSedimentary_Blend_NS}.Result
   )
   False=NamedRerouteDeclaration(None=SetMaterialAttributes!(MaterialAttributes=@n17.MaterialAttributes, Normal=@n24) {attributeSetTypes [#6]}) {name=L1 Baked Rocks, nodeColor=<rgba 0.9219 1 0 1>}
   Value=NamedRerouteDeclaration(None=StaticBoolParameter() {group=0_Rock/Cliff Main Settings, parameterName=L2 is WS with Baked L1 N}) {name=Use Sedimentary Cliffs, nodeColor=<rgba 0.1953 1 0 1>}
  )
  n26 = Divide(
   A=ComponentMask(None=WorldPosition() {worldPositionShaderOffset=WPT_ExcludeAllShaderOffsets}.XYZ) {g=true, r=true}
   B=ScalarParameter() {defaultValue=700, group="A_L1_Rock Surface (TD slot 1)", parameterName=L1 Tile, sortPriority=-1}
  )
  n27 = TextureSampleParameter2D(UVs=@n26) {
   group="A_L1_Rock Surface (TD slot 1)"
   parameterName=Diffuse
   samplerSource=SSM_Wrap_WorldGroupSettings
   sortPriority=100
   texture=&T_Asteria_RockCliff_04_D
  }
  n28 = TextureSampleParameter2D(UVs=@n26) {
   group="A_L1_Rock Surface (TD slot 1)"
   parameterName=SpecularMasks
   samplerSource=SSM_Wrap_WorldGroupSettings
   samplerType=SAMPLERTYPE_Masks
   sortPriority=105
   texture=&T_Asteria_RockCliff_04_S
  }
  n29 = TextureSampleParameter2D(UVs=@n26) {
   group="A_L1_Rock Surface (TD slot 1)"
   parameterName=Normals
   samplerSource=SSM_Wrap_WorldGroupSettings
   samplerType=SAMPLERTYPE_Normal
   sortPriority=110
   texture=&T_Asteria_RockCliff_04_N
  }
  n30 = SetMaterialAttributes!(MaterialAttributes, BaseColor=@n27.RGB, Opacity=@n27.RGB, Roughness=@n28.B, Specular=@n28.B, Metallic=@n28.B, Normal=@n29.RGB) {attributeSetTypes [
    #1
    #2
    #3
    #4
    #5
    #6
   ]}
  n31 = GetMaterialAttributes(None=StaticSwitchParameter(True=@n30, False=NamedRerouteDeclaration(None=@n25) {name=MobileSedimentMaterial, nodeColor=<rgba 0.6875 1 0 1>}) {group="A_L1_Rock Surface (TD slot 1)", parameterName=L1 is WS_Z Projection, sortPriority=-1}) {attributeGetTypes [
    #1
    #4
    #5
    #6
   ]}
  n32 = SetMaterialAttributes!(MaterialAttributes=@n31.MaterialAttributes, BaseColor=NamedRerouteUsage(), Specular=NamedRerouteUsage(), Metallic=NamedRerouteUsage()) {attributeSetTypes [#1, #4, #5]}
  n33 = NamedRerouteDeclaration(None=SetMaterialAttributes!(
    MaterialAttributes=@n32
    BaseColor=StaticSwitchParameter(True=@n1, False=@n32) {group="A_L1_Rock Surface (TD slot 1)", parameterName=Debug L1 UV's, sortPriority=-100}
   ) {attributeSetTypes [#1]}) {name=Mobile Rock Material, nodeColor=<rgba 1 0 0.5859 1>}
  n34 = GetMaterialAttributes(None=@n33) {attributeGetTypes [#1, #4, #6]}
  n35 = TextureCoordinate()
  n36 = Multiply(
   A=Divide(A=ObjectRadius()) {constB=256}
   B=ScalarParameter() {defaultValue=1, group="B_L2_Cliff Details (TD slot 2)", parameterName=L2 Cliff Normal Scale, sortPriority=30}
  )
  n37 = TextureSampleParameter2D(UVs=NamedRerouteDeclaration(None=Multiply(A=@n35, B=@n36)) {name=L2 Cliff Uvs, nodeColor=<rgba 0 1 0.9531 1>}) {
   group=B_L2_Shared
   parameterName=Normals_Texture_2
   samplerSource=SSM_Wrap_WorldGroupSettings
   samplerType=SAMPLERTYPE_Normal
   texture=/Engine/EngineMaterials/FlatNormal
  }
  n38 = ComponentMask(None=@n37.RGB) {g=true, r=true}
  n39 = ScalarParameter() {defaultValue=1, group="B_L2_Cliff Details (TD slot 2)", parameterName=L2 Cliff Normal Value, sortPriority=30}
  AppendVector(A=Add(A=ComponentMask(None=@n34.Normal) {g=true, r=true}, B=Multiply(A=@n38, B=@n39)), B=Constant())
  n40 = Constant() {r=-0.25}
  n41 = Constant() {r=1.5}
  n42 = Clamp(None=Saturate(None=LinearInterpolate(A=@n40, B=@n41, Alpha=@n34.Specular)), Max=Constant() {r=0.5}) {clampMode=CMODE_ClampMax}
  n43 = MaterialFunctionCall(Texture Scale=ScalarParameter() {defaultValue=700, group="B_L2_Cliff Details (TD slot 2)", parameterName=Landscape Detail_Scale, sortPriority=1}) {materialFunction=&MF_TriplanarDithered}
  n44 = TextureSampleParameter2D!(UVs=@n43.UVs, "DDX(UVs)"=@n43.UVs, "DDY(UVs)"=@n43.UVs, Apply View MipBias) {
   group="B_L2_Cliff Details (TD slot 2)"
   mipValueMode=TMVM_Derivative
   parameterName=Landscape Detail_Texture
   samplerSource=SSM_Wrap_WorldGroupSettings
   sortPriority=0
   texture=/Game/Environments/Apollo/Rocks/Natural_Rocks/Textures/T_RockDetails_02D
  }
  n45 = Add(A=Constant() {r=-1.1}, B=Multiply(A=Constant() {r=2.05}, B=@n44.R))
  n46 = LinearInterpolate(A=Constant() {r=7}, B=Constant() {r=-0.8}, Alpha=@n44.R)
  n47 = ScalarParameter() {defaultValue=1, group="B_L2_Cliff Details (TD slot 2)", parameterName=Landscape Detail_Spec, sortPriority=3}
  n48 = LinearInterpolate(B=Multiply(A=Saturate(None=@n45), B=Saturate(None=@n46)), Alpha=@n47) {constA=1}
  n49 = SquareRoot(None=@n44.RGB)
  n50 = ScalarParameter() {defaultValue=1, group="B_L2_Cliff Details (TD slot 2)", parameterName=Landscape Detail_Color, sortPriority=2}
  n51 = Multiply(A=Multiply(A=@n34.BaseColor, B=LinearInterpolate(B=@n49, Alpha=@n50) {constA=1}), B=Constant() {r=1.5})
  n52 = Multiply(B=@n47) {constA=0.25}
  n53 = ShadingPathSwitch(
   Default=SetMaterialAttributes!(MaterialAttributes=@n34.MaterialAttributes, BaseColor=Saturate(None=@n51), Specular=Multiply(A=@n42, B=@n48)) {attributeSetTypes [#1, #4]}
   Mobile=SetMaterialAttributes!(MaterialAttributes=@n34.MaterialAttributes, Specular=Multiply(A=@n34.Specular, B=@n52)) {attributeSetTypes [#4]}
  )
  n54 = Multiply(A=@n37.RGB, B=Constant3Vector() {constant=<rgba 1 1 0 0>})
  n55 = Normalize(VectorInput=Add(A=@n34.Normal, B=Multiply(A=@n54, B=@n39)))
  n56 = SetMaterialAttributes!(
   MaterialAttributes=@n53
   Normal=StaticSwitch(
    True=@n34.Normal
    False=StaticSwitchParameter(True=@n55, False=@n34.Normal) {group="B_L2_Cliff Details (TD slot 2)", parameterName=L2 Cliff Normal Details, sortPriority=29}
    Value=NamedRerouteUsage()
   )
  ) {attributeSetTypes [#6]}
  n57 = TextureCoordinate() {coordinateIndex=1}
  n58 = TextureSampleParameter2D!(UVs=@n57, Apply View MipBias) {
   group=B_L2_Shared
   parameterName=Normals_Texture_2
   samplerSource=SSM_Wrap_WorldGroupSettings
   samplerType=SAMPLERTYPE_Normal
   texture=&FlatNormal
  }
  n59 = Transform(None=ShaderStageSwitch(
    PixelShader=MaterialFunctionCall!(UVs=@n57, Normal=@n58.RGB) {materialFunction=/Engine/Functions/Engine_MaterialFunctions01/Texturing/DeriveTangentBasis}.World Space Normal
    VertexShader=@n58.RGB
   )) {transformSourceType=TRANSFORMSOURCE_World, transformType=TRANSFORM_Tangent}
  n60 = ScalarParameter() {defaultValue=1.2, group="B_L2_Baked Details (TD slot 2)", parameterName="L2 Baked Normals Value Baked Prop (non-Nanite)", sortPriority=50}
  n61 = StaticSwitch(
   True=@n60
   False=ScalarParameter() {defaultValue=2, group="B_L2_Baked Details (TD slot 2)", parameterName="L2 Baked Normals Value (non-Nanite)", sortPriority=50}
   Value=NamedRerouteUsage()
  )
  n62 = ScalarParameter() {defaultValue=1, group="B_L2_Baked Details (TD slot 2)", parameterName="L2 Baked Normals Value (Nanite)", sortPriority=50}
  n63 = Multiply(A=ComponentMask(None=@n59) {g=true, r=true}, B=NaniteReplace(Default=Multiply(A=@n61, B=@n62) {constA=2}, Nanite=@n62))
  n64 = ComponentMask(None=@n59) {b=true}
  n65 = Constant3Vector() {constant=<rgba 1 1 0 0>}
  n66 = GetMaterialAttributes(None=@n33) {attributeGetTypes [
    #1
    #4
    #3
    #6
    #2
   ]}
  n67 = Normalize(VectorInput=Add(A=AppendVector(A=@n63, B=@n64), B=Multiply(A=@n66.Normal, B=@n65)))
  n68 = VectorParameter() {
   group="B_L2_Baked Details (TD slot 2)"
   parameterName="Dirt_AO Color ( Expand for blending option)"
   channelNames {a=Linear vs Mult, b="", g="", r=""}
   defaultValue=<rgba 0.2691 0.1783 0.1117 0.5>
  }
  n69 = Constant() {r=1}
  n70 = ScalarParameter() {
   defaultValue=4
   desc="Multiplies SpecularMasks_2 G Channel (L2 thickness mask)"
   group="B_L2_Baked Details (TD slot 2)"
   parameterName=Dirt_Edge Mask Multiplier
   sortPriority=31
  }
  n71 = Multiply(A=@n70, B=@n69) {constA=4}
  n72 = TextureSampleParameter2D(UVs=@n57) {
   group=B_L2_Shared
   parameterName=SpecularMasks_2
   samplerSource=SSM_Wrap_WorldGroupSettings
   samplerType=SAMPLERTYPE_Masks
   texture=&T_Fortnite_Default_S
  }
  n73 = Multiply(A=Multiply(A=LinearInterpolate(A=@n69, B=@n71, Alpha=@n72.G), B=@n66.BaseColor), B=NamedRerouteUsage())
  n74 = Multiply(A=@n68, B=@n73)
  n75 = Multiply(A=LinearInterpolate(A=@n66.Specular, Alpha=NamedRerouteUsage()), B=@n72.R)
  n76 = ScalarParameter() {
   defaultValue=0.75
   desc=Also Darkens Basecolor on the bottom half of the props using the L1 roughtns map
   group="B_L2_Baked Details (TD slot 2)"
   parameterName=Mask AO_Roughness and Basecolor
  }
  n77 = ScalarParameter() {
   defaultValue=2
   desc=Also Darkens Basecolor on the bottom half of the props using the L1 roughtns map
   group="B_L2_Baked Details (TD slot 2)"
   parameterName=Mask Edges_Roughness and Basecolor
  }
  n78 = NamedRerouteUsage()
  n79 = SetMaterialAttributes!(
   MaterialAttributes=@n66.MaterialAttributes
   BaseColor=LinearInterpolate(A=LinearInterpolate(A=@n74, B=@n68, Alpha=@n68), B=@n73, Alpha=Saturate(None=@n75)) {constA=1}
   Specular=Multiply(A=@n66.Specular, B=NamedRerouteUsage())
   Roughness=Multiply(A=@n66.Roughness, B=LinearInterpolate(A=@n76, B=@n77, Alpha=@n78))
   Normal=StaticSwitch(True=@n66.Normal, False=ShadingPathSwitch(Default=@n67, Mobile=@n66.Normal), Value=NamedRerouteUsage())
  ) {attributeSetTypes [
    #1
    #4
    #3
    #6
   ]}
  n80 = GetMaterialAttributes(None=@n79) {attributeGetTypes [#1, #3]}
  n81 = NamedRerouteUsage()
  n82 = OneMinus(None=@n80.Roughness)
  n83 = Saturate(None=LinearInterpolate(A=Constant() {desc=-2.5, r=-3.75}, B=Constant() {desc=2.5, r=3.75}, Alpha=Subtract(A=@n81, B=@n82)))
  n84 = LinearInterpolate(A=Multiply(A=Add(B=@n80.BaseColor) {constA=0.5}, B=@n80.BaseColor), B=@n80.BaseColor, Alpha=@n83)
  n85 = Multiply(A=@n80.Roughness) {constB=2}
  n86 = StaticSwitch(
   True=SetMaterialAttributes!(MaterialAttributes=@n80.MaterialAttributes, BaseColor=Saturate(None=@n84), Roughness=LinearInterpolate(A=@n85, B=@n80.Roughness, Alpha=@n83)) {attributeSetTypes [#1, #3]}
   False=MaterialFunctionCall!(
    UV's=NamedRerouteUsage()
    Checker true=StaticBoolParameter() {group="B_L2_Cliff Details (TD slot 2)", parameterName=Debug L2 Cliff UVs, sortPriority=-10}
    Material In=QualitySwitch(Default=@n53, Low=@n53, High=@n53, Medium=@n53, Epic=@n56)
   ) {materialFunction=&/MF_Checker}.Switch
   Value=NamedRerouteUsage()
  )
  n87 = StaticSwitch(True=ShadingPathSwitch(Default=@n86, Mobile=NamedRerouteUsage()), False=@n86, Value=NamedRerouteUsage())
  n88 = StaticSwitchParameter(True=BlendMaterialAttributes(A=@n87, B=NamedRerouteUsage(), Alpha=NamedRerouteUsage()) {vertexAttributeBlendType=UseB}, False=@n87) {
   desc="If set to True, set TopTerrainMaterial and TopSnow to False"
   group=0_Rock/Cliff Main Settings
   parameterName=Custom Top Layer
   sortPriority=10
  }
  n89 = GetMaterialAttributes(None=@n88) {attributeGetTypes [#6]}
  n90 = SetMaterialAttributes!(MaterialAttributes=@n89.MaterialAttributes, Normal=Transform(None=@n89.Normal)) {attributeSetTypes [#6]}
  n91 = MaterialFunctionCall!(In=@n90) {materialFunction=/Game/Environments/Helios/Foliage/Trees/MasterMaterials/MaterialFunctions/MF_WetlandsMoss_World}
  n92 = StaticSwitchParameter(
   True=BlendMaterialAttributes(A=@n91.Result, B=NamedRerouteUsage(), Alpha=NamedRerouteUsage()) {vertexAttributeBlendType=UseB}
   False=@n91.Result
  ) {desc="If set to True, set TopLayer to False", group=0_Rock/Cliff Main Settings, parameterName=Top Terrain Material, sortPriority=11}
  n93 = NamedRerouteUsage()
  n94 = NamedRerouteUsage()
  n95 = MaterialFunctionCall!(
   In=StaticSwitchParameter(True=NamedRerouteUsage(), False=@n92) {group=0_Rock/Cliff Main Settings, parameterName=Top Terrain Only, sortPriority=12}
   "RVT Height Map (Roughness)"=StaticSwitch(True=Multiply(A=@n93, B=@n94), False=@n94, Value=NamedRerouteUsage())
  ) {desc="RVT ", materialFunction=&/MF_RVTBlendingRocks}
  n96 = MaterialFunctionCall!(RVT=NamedRerouteUsage(), WaterInput=@n95.Result) {materialFunction=&/MF_Rocks_2023_Waterline}
  n97 = MaterialFunctionCall!(Material=@n95.Result) {materialFunction=/Game/Athena/Helios/Environments/Landscape/Materials/MaterialFunctions_Utilities/MF_WetShore_Mobile}
  n98 = NamedRerouteUsage()
  n99 = StaticSwitchParameter(True=NamedRerouteUsage(), False=Constant()) {group=0_Rock/Cliff Main Settings, parameterName=Custom Top Layer, sortPriority=10}
  n100 = StaticSwitchParameter(
   True=BlendMaterialAttributes(
    A=ShadingPathSwitch(Default=@n96.BlendMaterialAttributes, Mobile=@n97.Result)
    B=@n95.Result
    Alpha=StaticSwitchParameter(True=@n98, False=@n99) {group=0_Rock/Cliff Main Settings, parameterName=Top Terrain Material, sortPriority=11}
   ) {vertexAttributeBlendType=UseB}
   False=@n95.Result
  ) {group=0_Rock/Cliff Main Settings, parameterName=Use Waterline}
  Composite!(MaterialAttributes=@n100) {subgraphName=Documentation Link}
  n101 = MaterialFunctionCall!(Grass_UseCustomSlopeMask, Grass_CustomSlopeMask) {materialFunction=/Game/Athena/Helios/Environments/Landscape/Materials/MaterialFunctions_Utilities/MF_Helios_StaticMeshLandscapeLayerMatch}
  n102 = GetMaterialAttributes(None=@n101.SnowLayer) {attributeGetTypes [#7, #6]}
  n103 = NamedRerouteUsage()
  n104 = NamedRerouteUsage()
  Composite!(MaterialAttributes=@n102.MaterialAttributes, Snow Normals=@n102.MaterialAttributes, L1 Normals=@n103, IsBaked=@n104) {subgraphName=Snow Normal Blend}
  MaterialExpressionReroute_70 = Reroute(None=@n100)
  n105 = ScalarParameter() {
   desc=Documentation link
   group="0 (Expand)  Documentation (RMB - Copy display name) and Debug view"
   parameterName=https://confluence.it.epicgames.com/display/FORT/Fortnite+Base+Material+2022
  }
  n106 = SetMaterialAttributes!(MaterialAttributes=@MaterialExpressionReroute_70.MaterialAttributes, WorldPositionOffset=Multiply(A=@n105)) {attributeSetTypes [#8]}
  MaterialExpressionReroute_98 = Reroute(None=@n106)
  Composite!(MaterialAttributes=@MaterialExpressionReroute_98.Output) {desc=Game Shared WPO and Fire Effect, subgraphName=WPO And Shared Functions}
  Constant() {r=8}
  Constant() {r=2}
  Constant() {r=10}
  n107 = StaticComponentMaskParameter(None=VertexColor()) {defaultR=true, group=D_L3 Custom Mask, parameterName=L3 Mask_Vertex Channel, sortPriority=2}
  n108 = VectorParameter() {
   desc=Modulates VTX color transition intensity
   group=D_L3 Custom Mask
   parameterName=L3Mask_VTX Color Contrast
   sortPriority=1
   defaultValue=<rgba 0 0 0 1>
  }
  n109 = Constant() {r=2}
  DotProduct(
   A=@n107
   B=ShadingPathSwitch(
    Default=@n108.RGB
    Mobile=Multiply(A=@n108.RGB, B=@n109) {desc=compensates for the less dense meshes on mobile creating bigger vertex gradients}
   )
  )
  n110 = VertexColor()
  MaterialExpressionReroute_99 = Reroute(None=@MaterialExpressionReroute_98.Output)
  n111 = MaterialFunctionCall!(
   In=MaterialFunctionCall(Material_In=@MaterialExpressionReroute_99.MaterialAttributes) {materialFunction=&MF_FN_DestructibleDamage}.Material_Out
   UseSphericalMappingOverride
  ) {materialFunction=/Game/Packages/Fortress_SharedMaterials/Base_Material_Functions/MF_CurieFallbackEffect}
  MaterialExpressionReroute_100 = Reroute(None=@n111.Result)
  FunctionOutput(None=StaticSwitchParameter(
    True=SetMaterialAttributes!(MaterialAttributes, EmissiveColor=@n110) {attributeSetTypes [#9]}
    False=NamedRerouteDeclaration(None=@MaterialExpressionReroute_100.Result) {name=MA Final, nodeColor=<rgba 1 0.2109 0 1>}
   ) {group=D_TopMask, parameterName="[DEBUG] Vertex Colors", sortPriority=-1})
  MaterialFunctionCall!(In=@n90) {materialFunction=/Game/Environments/Asteria/Foliage/Trees/MasterMaterials/MaterialFunctions/MF_JungleMoss_World}
  Multiply(A=ComponentMask(None=@n34.Normal) {b=true}, B=ComponentMask(None=@n37.RGB) {b=true})
  n112 = ScalarParameter() {defaultValue=2.5, group="B_L2_Baked Details (TD slot 2)", parameterName=VerticalGradient_Contrast}
  Multiply(A=@n112) {constB=-1}
  Multiply(A=Subtract(A=@n66.Opacity, B=@n72.A))
  NamedRerouteDeclaration(None=StaticSwitchParameter(True=@n107, False=OneMinus(None=@n107)) {group=D_L3 Custom Mask, parameterName=L3 Mask_Invert Vertex mask, sortPriority=2}) {name=L3Mask_Vertex Color, nodeColor=<rgba 0 0.8125 1 1>}
  NamedRerouteDeclaration(None=@n95.RVT H) {name=RVT H, nodeColor=<rgba 1 0.5156 0 1>}
  NamedRerouteDeclaration(None=@n31.Metallic) {name=L1_Metallic, nodeColor=<rgba 0 0.2031 1 1>}
  n113 = StaticBoolParameter() {
   desc="True - Baked Props Will Use WS projection\r\nFalse - Baked Props Use UV Map"
   group="C_L3_Top (TD slot 3)"
   parameterName="L3 Custom Top_Use WS (Defaults to Z only)"
   sortPriority=-2
  }
  n114 = ObjectRadius()
  n115 = Multiply(
   A=Multiply(
    A=ScalarParameter() {defaultValue=1, group="C_L3_Top (TD slot 3)", parameterName=L3 Baked Tile, sortPriority=0}
    B=Divide(A=@n114) {constB=256}
   )
   B=TextureCoordinate()
  )
  n116 = TextureSampleParameter2D(UVs=@n115) {
   group="C_L3_Top (TD slot 3)"
   parameterName=Diffuse_Texture_3
   samplerSource=SSM_Wrap_WorldGroupSettings
   sortPriority=1
   texture=/Game/Global/Textures/Default/Blanks/T_MidgreysRGB
  }
  n117 = MaterialFunctionCall(Texture Scale=ScalarParameter() {defaultValue=2000, group="C_L3_Top (TD slot 3)", parameterName=L3 WS_Tile, sortPriority=-1}) {materialFunction=&MF_TriplanarDithered}
  n118 = TextureSampleParameter2D!(UVs=@n117.UVs, "DDX(UVs)"=@n117.DDX, "DDY(UVs)"=@n117.DDY, Apply View MipBias) {
   group="C_L3_Top (TD slot 3)"
   mipValueMode=TMVM_Derivative
   parameterName=Diffuse_Texture_3
   samplerSource=SSM_Wrap_WorldGroupSettings
   sortPriority=1
   texture=&T_MidgreysRGB
  }
  n119 = ShadingPathSwitch(Default=@n118.RGBA, Mobile=@n116.RGBA)
  n120 = NamedRerouteUsage()
  n121 = StaticSwitch(True=StaticSwitch(True=@n119, False=@n116.RGBA, Value=@n113), False=@n119, Value=@n120)
  NamedRerouteDeclaration(None=MaterialFunctionCall(
    Transition Phase=NamedRerouteUsage()
    Height Texture=ComponentMask(None=@n121) {a=true}
    Contrast=ScalarParameter() {group="C_L3_Top (TD slot 3)", parameterName=L3 Height Top Contrast, sortPriority=2}
   ) {materialFunction=&HeightLerp}.Alpha) {name=Custom Top H, nodeColor=<rgba 0.4531 1 0 1>}
  NamedRerouteDeclaration(None=@n31.Normal) {name=L1 Normals, nodeColor=<rgba 0.8984 0 1 1>}
  NamedRerouteDeclaration(None=@n101.OasisCliffMask) {name=OasisCliffMask, nodeColor=<rgba 0.5 0 1 1>}
  n122 = NamedRerouteUsage()
  n123 = LinearInterpolate(
   A=@n122
   B=ScalarParameter() {defaultValue=0.5, group=A_L1_G Channel Spec Mask Modifier, parameterName=S_G Channel Mask_Specular Value, sortPriority=12}
   Alpha=NamedRerouteUsage()
  )
  n124 = NamedRerouteUsage()
  NamedRerouteDeclaration(None=StaticSwitch(True=@n122, False=StaticSwitch(True=@n123, False=@n122, Value=@n124), Value=NamedRerouteUsage())) {name=L1 Specular Adjusted, nodeColor=<rgba 0.2422 0 1 1>}
  n125 = GetMaterialAttributes(None=@n87) {attributeGetTypes [#6]}
  n126 = SetMaterialAttributes!(
   MaterialAttributes=@n125.MaterialAttributes
   BaseColor=NamedRerouteUsage()
   Specular=NamedRerouteUsage()
   Roughness=NamedRerouteUsage()
   Normal=NamedRerouteUsage()
   Metallic=NamedRerouteUsage()
  ) {attributeSetTypes [
    #1
    #4
    #3
    #6
    #5
   ]}
  NamedRerouteDeclaration(None=@n126) {name=Custom Top MA, nodeColor=<rgba 1 0.3044 0 1>}
  n127 = TextureSampleParameter2D!(UVs=@n115, Apply View MipBias) {
   group="C_L3_Top (TD slot 3)"
   parameterName=Normals_Texture_3
   samplerSource=SSM_Wrap_WorldGroupSettings
   samplerType=SAMPLERTYPE_Normal
   sortPriority=20
   texture=&FlatNormal
  }
  n128 = TextureSampleParameter2D!(UVs=@n117.UVs, "DDX(UVs)"=@n117.DDX, "DDY(UVs)"=@n117.DDY, Apply View MipBias) {
   group="C_L3_Top (TD slot 3)"
   mipValueMode=TMVM_Derivative
   parameterName=Normals_Texture_3
   samplerSource=SSM_Wrap_WorldGroupSettings
   samplerType=SAMPLERTYPE_Normal
   sortPriority=20
   texture=&FlatNormal
  }
  n129 = ShadingPathSwitch(
   Default=Transform(None=MaterialFunctionCall(Tangent Normals=@n128.RGB, Triplanar Mask=@n117.Mask) {materialFunction=&MF_TriplanarNormalTangentToWorld}.World Normals) {transformSourceType=TRANSFORMSOURCE_World, transformType=TRANSFORM_Tangent}
   Mobile=@n127.RGB
  )
  n130 = StaticSwitch(True=StaticSwitch(True=@n129, False=@n127.RGB, Value=@n113), False=@n129, Value=@n120)
  n131 = AppendVector(
   A=Multiply(
    A=ScalarParameter() {defaultValue=1, group="C_L3_Top (TD slot 3)", parameterName=L3 Normals Value, sortPriority=21}
    B=ComponentMask(None=@n130) {g=true, r=true}
   )
   B=ComponentMask(None=@n130) {b=true}
  )
  n132 = Constant3Vector() {constant=<rgba 1 1 0 0>}
  n133 = LinearInterpolate(
   A=@n131
   B=Add(A=Multiply(A=@n125.Normal, B=@n132), B=@n131)
   Alpha=Saturate(None=ScalarParameter() {defaultValue=1, group="C_L3_Top (TD slot 3)", parameterName=Add Normals Value, sortPriority=24})
  )
  NamedRerouteDeclaration(None=Normalize(VectorInput=StaticSwitchParameter(True=@n133, False=@n131) {group="C_L3_Top (TD slot 3)", parameterName=Add L3 to L2 Normals, sortPriority=23})) {name=CustomTopNormal, nodeColor=<rgba 1 0 0.7734 1>}
  n134 = NamedRerouteUsage()
  n135 = Add(A=OneMinus(None=Multiply(A=@n134) {constB=0.009})) {constB=-0.5}
  n136 = ComponentMask(None=NamedRerouteUsage()) {g=true}
  n137 = Add(A=Multiply(A=Divide(A=@n135, B=@n136)) {constB=100}) {constB=-30}
  n138 = Multiply(A=ComponentMask(None=NamedRerouteUsage()) {r=true}) {constB=20}
  NamedRerouteDeclaration(None=Min(A=Saturate(None=@n137), B=Add(A=@n138) {constB=-2})) {name=Snow Biom RVTColor Mask, nodeColor=<rgba 1 0.7969 0 1>}
  n139 = TextureSampleParameter2D!(UVs=@n115, Apply View MipBias) {
   group="C_L3_Top (TD slot 3)"
   parameterName=SpecularMasks_3
   samplerSource=SSM_Wrap_WorldGroupSettings
   samplerType=SAMPLERTYPE_Masks
   sortPriority=10
   texture=&T_Fortnite_Default_S
  }
  n140 = TextureSampleParameter2D!(UVs=@n117.UVs, "DDX(UVs)"=@n117.DDX, "DDY(UVs)"=@n117.DDY, Apply View MipBias) {
   group="C_L3_Top (TD slot 3)"
   mipValueMode=TMVM_Derivative
   parameterName=SpecularMasks_3
   samplerSource=SSM_Wrap_WorldGroupSettings
   samplerType=SAMPLERTYPE_Masks
   sortPriority=10
   texture=&T_Fortnite_Default_S
  }
  n141 = ShadingPathSwitch(Default=@n140.RGB, Mobile=@n139.RGB)
  n142 = MaterialFunctionCall(Float3=StaticSwitch(True=StaticSwitch(True=@n141, False=@n139.RGB, Value=@n113), False=@n141, Value=@n120)) {materialFunction=&BreakOutFloat3Components}
  NamedRerouteDeclaration(None=@n142.B) {name=CustomTopRoughness, nodeColor=<rgba 1 0.5859 0 1>}
  n143 = WorldPosition()
  n144 = Add(A=Multiply(A=Add(A=@n143.Z) {constB=30}) {constB=0.015}, B=NamedRerouteUsage())
  NamedRerouteDeclaration(None=Saturate(None=Subtract(A=@n144))) {name=Snow_Waterline, nodeColor=<rgba 1 0.8438 0 1>}
  NamedRerouteDeclaration(None=@n83) {name=DebugRerout, nodeColor=<rgba 0 0 0 1>}
  MaterialExpressionReroute_50 = Reroute(None=@n103)
  MaterialExpressionReroute_49 = Reroute(None=@n102.Normal)
  n145 = Add(A=@MaterialExpressionReroute_49.Snow Normals, B=Multiply(A=@MaterialExpressionReroute_50.L1 Normals) {constB=0.5})
  MaterialExpressionReroute_48 = Reroute(None=@n102.MaterialAttributes)
  n146 = SetMaterialAttributes!(
   MaterialAttributes=@MaterialExpressionReroute_48.MaterialAttributes
   Normal=Normalize(VectorInput=@MaterialExpressionReroute_49.Snow Normals)
  ) {attributeSetTypes [#6]}
  MaterialExpressionReroute_27 = Reroute(None=@n104)
  n147 = StaticSwitch(
   True=SetMaterialAttributes(MaterialAttributes=@n146, Normal=Normalize(VectorInput=@n145)) {attributeSetTypes [#6]}
   False=@n146
   Value=@MaterialExpressionReroute_27.IsBaked
  )
  MaterialExpressionReroute_51 = Reroute(None=@n147)
  n148 = GetMaterialAttributes(None=@n101.DesertOasisLayer) {attributeGetTypes [#7]}
  n149 = GetMaterialAttributes(None=@n101.DesertGrassLayer) {attributeGetTypes [#7]}
  n150 = GetMaterialAttributes(None=@n101.AsphaltLayer) {attributeGetTypes [#7]}
  n151 = GetMaterialAttributes(None=@n101.SandLayer) {attributeGetTypes [#7]}
  n152 = GetMaterialAttributes(None=@n101.GravelLayer) {attributeGetTypes [#7]}
  n153 = GetMaterialAttributes(None=@n101.Brimstone Layer) {attributeGetTypes [#7]}
  n154 = GetMaterialAttributes(None=@n101.Grass_ChaparralCoast) {attributeGetTypes [#7]}
  n155 = GetMaterialAttributes(None=@n101.GrassLayer) {attributeGetTypes [#7]}
  n156 = StaticBoolParameter() {desc="If set to True, set Custom Top Layer to False", group=C_L3_Top Terrain, parameterName=Top Chaparral Grass, sortPriority=-4}
  n157 = StaticBoolParameter() {group=C_L3_Top Terrain, parameterName=Top Brimstone, sortPriority=-4}
  n158 = StaticBoolParameter() {group=C_L3_Top Terrain, parameterName=TopGravel, sortPriority=-4}
  n159 = StaticSwitch(
   True=@n152.MaterialAttributes
   False=StaticSwitch(True=@n153.MaterialAttributes, False=StaticSwitch(True=@n154.MaterialAttributes, False=@n155.MaterialAttributes, Value=@n156), Value=@n157)
   Value=@n158
  )
  n160 = StaticBoolParameter() {group=C_L3_Top Terrain, parameterName=TopSand, sortPriority=-4}
  n161 = StaticBoolParameter() {group=C_L3_Top Terrain, parameterName=TopAsphalt, sortPriority=-4}
  n162 = StaticBoolParameter() {group=C_L3_Top Terrain, parameterName=Top Desert Grass, sortPriority=-4}
  n163 = StaticSwitch(
   True=@n149.MaterialAttributes
   False=StaticSwitch(True=@n150.MaterialAttributes, False=StaticSwitch(True=@n151.MaterialAttributes, False=@n159, Value=@n160), Value=@n161)
   Value=@n162
  )
  n164 = StaticBoolParameter() {group=C_L3_Top Terrain, parameterName=Top Desert Oasis, sortPriority=-4}
  n165 = StaticBoolParameter() {desc="If set to True, set TopLayer to False", group=C_L3_Top Terrain, parameterName=TopSnow, sortPriority=-4}
  NamedRerouteDeclaration(None=StaticSwitch(True=@MaterialExpressionReroute_51.Output, False=StaticSwitch(True=@n148.MaterialAttributes, False=@n163, Value=@n164), Value=@n165)) {name=Biom MA, nodeColor=<rgba 0 1 0.1701 1>}
  n166 = StaticSwitch(True=@n154.OpacityMask, False=@n155.OpacityMask, Value=@n156)
  n167 = StaticSwitch(
   True=@n151.OpacityMask
   False=StaticSwitch(True=@n152.OpacityMask, False=StaticSwitch(True=@n153.OpacityMask, False=@n166, Value=@n157), Value=@n158)
   Value=@n160
  )
  n168 = StaticSwitch(
   True=@n148.OpacityMask
   False=StaticSwitch(True=@n149.OpacityMask, False=StaticSwitch(True=@n150.OpacityMask, False=@n167, Value=@n161), Value=@n162)
   Value=@n164
  )
  NamedRerouteDeclaration(None=LinearInterpolate(
    A=ScalarParameter() {group=C_L3_Top Terrain, parameterName=Terrain H Min}
    B=ScalarParameter() {defaultValue=1, group=C_L3_Top Terrain, parameterName=Terrain H Max}
    Alpha=StaticSwitch(True=@n102.OpacityMask, False=@n168, Value=@n165)
   )) {name=Biom H, nodeColor=<rgba 0 1 0.1701 1>}
  NamedRerouteDeclaration(None=@n142.G) {name=CustomTopMetallic, nodeColor=<rgba 1 0.9844 0 1>}
  NamedRerouteDeclaration(None=@n95.RVT Color) {name=RVT Color, nodeColor=<rgba 0 1 0.2734 1>}
  NamedRerouteDeclaration(None=@n142.R) {name=CustomTopSpecular, nodeColor=<rgba 1 0.0234 0 1>}
  NamedRerouteDeclaration(None=ComponentMask(None=@n121) {b=true, g=true, r=true}) {name=CustomTopBaseColor, nodeColor=<rgba 0 0.2969 1 1>}
  n169 = DistanceFieldApproxAO!(
   World Position
   Normal
   BaseDistance=ScalarParameter() {defaultValue=55, group=DFAO, parameterName=AO Distance}
   Radius=ScalarParameter() {defaultValue=150, group=DFAO, parameterName=AO Radius}
  ) {numSteps=3}
  n170 = Constant()
  n171 = QualitySwitch(Default=@n169, Low=@n170, High=@n170, Medium=@n170, Epic=@n169)
  n172 = Constant()
  NamedRerouteDeclaration(None=OneMinus(None=ShadingPathSwitch(Default=@n171, Mobile=@n172) {desc=Added to fix compile errors on Mobile})) {name=AO, nodeColor=<rgba 1 0 0.3516 1>}
  n173 = VectorParameter() {parameterName=Curvature Color, defaultValue=<rgba 0.4115 0.4115 0.4115 1>}
  n174 = ScalarParameter() {defaultValue=1, parameterName="Add Soft Curvature "}
  n175 = TextureSampleParameter2D!(UVs=TextureCoordinate() {coordinateIndex=1}, Apply View MipBias) {group=Fallback, parameterName=Fallback AO, texture=/Game/Global/Textures/Default/Blanks/T_Blank_LegalWhite_sRGBColor}
  n176 = ScalarParameter() {defaultValue=1, parameterName="Add Soft Curvature "}
  n177 = VertexColor()
  n178 = NaniteReplace(Default=Saturate(None=Multiply(A=@n174, B=@n175.G)), Nanite=Saturate(None=Multiply(A=@n176, B=@n177.G)))
  NamedRerouteDeclaration(None=Add(A=NamedRerouteUsage(), B=LinearInterpolate(B=@n173.RGB, Alpha=@n178))) {name=Curvature, nodeColor=<rgba 0 1 0.6719 1>}
  NamedRerouteDeclaration(None=@n31.Specular) {name=L1_Specular, nodeColor=<rgba 0.6641 1 0 1>}
  n179 = StaticBoolParameter() {
   desc="L1 Specular G channel wll be used as Metallic Mask.\r\nThis toggle overrides other L1 Specular Value Adjustments"
   group="A_L1_Rock Surface (TD slot 1)"
   parameterName=G Channel IS Metal
   sortPriority=11
  }
  NamedRerouteDeclaration(None=StaticSwitch(True=NamedRerouteUsage(), False=Constant(), Value=NamedRerouteDeclaration(None=@n179) {name=Spec_G IS Metal, nodeColor=<rgba 0 1 0.9531 1>})) {name=L1 Metallic Adjusted, nodeColor=<rgba 1 0 0.9844 1>}
  n180 = StaticBoolParameter() {defaultValue=true, group=1_Rock/Cliff Addons, parameterName="Use L1 Spec_G as Mask for Tint && Spec", sortPriority=-1}
  n181 = NamedRerouteUsage()
  n182 = ScalarParameter() {group=A_L1_G Channel Spec Mask Modifier, parameterName=Spec_G Mask Desaturate, sortPriority=11}
  n183 = VectorParameter() {
   group=A_L1_G Channel Spec Mask Modifier
   parameterName=Spec_G Mask Tint
   sortPriority=2
   channelNames {a=Blend to color, b="", g="", r=""}
   defaultValue=<rgba 1 1 1 0>
  }
  n184 = LinearInterpolate(A=Multiply(A=Desaturation(None=@n181, Fraction=@n182), B=@n183), B=@n183, Alpha=@n183)
  n185 = NamedRerouteUsage()
  n186 = NamedRerouteUsage()
  n187 = Desaturation(None=@n186, Fraction=ScalarParameter() {group="A_L1_Rock Surface (TD slot 1)", parameterName=L1 Desaturate, sortPriority=2})
  n188 = VectorParameter() {
   group="A_L1_Rock Surface (TD slot 1)"
   parameterName=L1 Tint
   sortPriority=3
   channelNames {a=Blend to color, b="", g="", r=""}
   defaultValue=<rgba 1 1 1 0>
  }
  n189 = StaticSwitch(
   True=LinearInterpolate(A=Multiply(A=@n187, B=@n188), B=@n188, Alpha=@n188)
   False=@n186
   Value=StaticBoolParameter() {defaultValue=true, group="A_L1_Rock Surface (TD slot 1)", parameterName=Color Tint, sortPriority=1}
  )
  NamedRerouteDeclaration(None=StaticSwitch(
    True=LinearInterpolate(A=@n189, B=@n184, Alpha=@n185)
    False=@n189
    Value=NamedRerouteDeclaration(None=@n180) {name=Use L1_Spec_G for Tint/Spec, nodeColor=<rgba 1 0 0.9844 1>}
   )) {name=L1 Base Color Adjusted, nodeColor=<rgba 0 1 0.6953 1>}
  NamedRerouteDeclaration(None=@n31.BaseColor) {name=L1_BaseColor, nodeColor=<rgba 1 0.375 0 1>}
  n190 = VertexNormalWS()
  n191 = ComponentMask(None=Transform(None=GetMaterialAttributes(None=@n87) {attributeGetTypes [#6]}.Normal)) {b=true}
  n192 = Power(
   Base=StaticSwitchParameter(True=ComponentMask(None=@n190) {b=true}, False=Saturate(None=@n191)) {desc=Defaults to Pixel Z Normals, parameterName=L3 Mask_VTX Only}
   Exp=ScalarParameter() {defaultValue=1, group=D_L3 Custom Mask, parameterName=L3 Mask_TextureNormalZ_Power, sortPriority=6}
  )
  n193 = ScalarParameter() {defaultValue=1, group=D_L3 Custom Mask, parameterName=L3 Mask_TextureNormalZ_Mult, sortPriority=7}
  NamedRerouteDeclaration(None=Saturate(None=Multiply(A=@n192, B=@n193))) {name=L3_Normal Z, nodeColor=<rgba 0 0.9062 1 1>}
  n194 = VectorParameter() {
   group=Cliff
   parameterName="Sediment_a | Pos && Scale"
   sortPriority=306
   channelNames {a=Scale, b=Z, g=Y, r=X}
   defaultValue=<rgba 0 0 0 2500>
  }
  n195 = VectorParameter() {
   group=Cliff
   parameterName="Sediment_a | Wave Dist Pos && Freq"
   sortPriority=307
   channelNames {a=Freq, b=Z, g=Y, r=X}
   defaultValue=<rgba 0 0 0 10000>
  }
  n196 = MaterialFunctionCall!(
   WorldPosition
   Sediment L2_WS | Position Offset_1=@n194
   Sediment L2_WS | Scale=@n194
   Use Wavy Distortion=StaticBoolParameter() {group=Cliff, parameterName=Sediment_a | Use Wave Distortion, sortPriority=306}
   Wave Distortion Intensity=ScalarParameter() {defaultValue=5000, group=Cliff, parameterName=Sediment_a | Wave Dist Intensity, sortPriority=307}
   Sediment L2_WS | Wave DistPos Offset=@n195
   Wave Distortion Freq=@n195
  ) {materialFunction=&/MF_Cliffs_SedimentWS_UVs}
  n197 = TextureObjectParameter() {group=B_L2_Shared, parameterName=Diffuse_Texture_2, texture=&T_Asteria_RockCliff_04_D}
  n198 = TextureObjectParameter() {
   group=B_L2_Shared
   parameterName=SpecularMasks_2
   samplerType=SAMPLERTYPE_Masks
   texture=&T_Fortnite_Default_S
  }
  n199 = TextureObjectParameter() {group=B_L2_Shared, parameterName=Normals_Texture_2, samplerType=SAMPLERTYPE_Normal, texture=&FlatNormal}
  n200 = ScalarParameter() {defaultValue=0.2, desc=XY Projection Transition Contrast, group=Cliff, parameterName=Sediment_a | PTC, sortPriority=306}
  n201 = MaterialFunctionCall!(UV_XZ=@n196.XZ, UV_YZ=@n196.XZ, T_Diffuse=@n197, T_Specular=@n198, T_Normal=@n199, PTC=@n200) {materialFunction=&/MF_Cliffs_SedimentWS}
  n202 = NamedRerouteUsage()
  n203 = StaticBool() {value=true}
  NamedRerouteDeclaration(None=ShadingPathSwitch(
    Default=@n201.SetMaterialAttributes
    Mobile=MaterialFunctionCall!(OasisMask=@n202, VariationMasks Mobile, IsStaticMesh=@n203) {materialFunction=/Game/Athena/Helios/Environments/Landscape/Materials/QualityLevelLayers/MF_Helios_CliffSedimentary_NS}.Result
   )) {name=Sediment MA, nodeColor=<rgba 0.6172 0 1 1>}
  NamedRerouteDeclaration(None=@n201.Sediment Normal B Mask) {name=Sediment Normal B mask, nodeColor=<rgba 1 0 0.8906 1>}
  NamedRerouteDeclaration(None=@n196.Sediment Wave Distortion) {name=Sediment Wave Distortion, nodeColor=<rgba 0 1 0.6484 1>}
  NamedRerouteDeclaration(None=@n2.Sediment Bands Mask) {name=Sediment Bands Mask, nodeColor=<rgba 0 1 0.7891 1>}
  NamedRerouteDeclaration(None=ShadingPathSwitch(Default=@n96.Waterline Height Mask, Mobile=@n97.Waterline Height Mask)) {name=Mask_Waterline, nodeColor=<rgba 0.2422 1 0 1>}
  n204 = TextureSample!(UVs=@n5.UVs, Tex, "DDX(UVs)"=@n5.DDX, "DDY(UVs)"=@n5.DDY, Apply View MipBias) {
   mipValueMode=TMVM_Derivative
   samplerSource=SSM_Wrap_WorldGroupSettings
   texture=&Checker
  }
  NamedRerouteDeclaration(None=@n204.RGB) {name=L1 WS_UV Debug, nodeColor=<rgba 0 1 0.4844 1>}
  n205 = VectorParameter() {group=D_TopMask, parameterName=VertexColor Mask for Grass, sortPriority=1, defaultValue=<rgba 0 0 0 1>}
  n206 = Constant() {r=2}
  n207 = VertexColor()
  n208 = DotProduct(
   A=ShadingPathSwitch(
    Default=@n205.RGB
    Mobile=Multiply(A=@n205.RGB, B=@n206) {desc=compensates for the less dense meshes on mobile creating bigger vertex gradients}
   )
   B=AppendVector(A=@n207, B=@n207)
  )
  n209 = StaticSwitchParameter(True=@n208, False=OneMinus(None=@n208)) {group=D_TopMask, parameterName=Invert Vertex mask, sortPriority=2}
  n210 = GetMaterialAttributes(None=@n87) {attributeGetTypes [#6]}
  n211 = Saturate(None=ComponentMask(None=Transform(None=@n210.Normal)) {b=true})
  n212 = ScalarParameter() {defaultValue=1, group=D_TopMask, parameterName=TopMask_TextureNormalZ_Power, sortPriority=6}
  n213 = Saturate(None=Multiply(
    A=Power(Base=@n211, Exp=@n212)
    B=ScalarParameter() {defaultValue=1, group=D_TopMask, parameterName=TopMask_TextureNormalZ_Mult, sortPriority=7}
   ))
  n214 = StaticSwitchParameter(
   True=StaticSwitchParameter(True=Max(A=@n209, B=@n213), False=@n209) {group=D_TopMask, parameterName="Add Top Masks (Pixel on Vertex)"}
   False=@n213
  ) {group=D_TopMask, parameterName=Use Top Mask Vertex Colors, sortPriority=0}
  n215 = ScalarParameter() {group=D_TopMask, parameterName=Top Contrast, sortPriority=6}
  n216 = Constant() {r=2.5}
  n217 = MaterialFunctionCall(
   Transition Phase=NamedRerouteDeclaration(None=Saturate(None=@n214)) {name=CustomTopLayer Blend Mask, nodeColor=<rgba 1 0.3323 0 1>}
   Height Texture=NamedRerouteUsage()
   Contrast=ShadingPathSwitch(Default=@n215, Mobile=Multiply(A=@n215, B=@n216) {desc=compensates for the less dense meshes on mobile creating bigger vertex gradients})
  ) {materialFunction=&HeightLerp}
  n218 = ShadingPathSwitch(Default=Max(A=@n217.Alpha, B=NamedRerouteUsage()), Mobile=@n217.Alpha)
  n219 = NamedRerouteUsage()
  NamedRerouteDeclaration(None=StaticSwitch(
    True=Min(A=@n218, B=@n219)
    False=@n217.Alpha
    Value=StaticBoolParameter() {desc="If set to True, set TopLayer to False", group=C_L3_Top Terrain, parameterName=TopSnow, sortPriority=-4}
   )) {name=TopBiom Blend Mask, nodeColor=<rgba 0 0.3437 1 1>}
  NamedRerouteDeclaration(None=MaterialFunctionCall!(Linear=Multiply(A=@n66.Opacity, B=@n72.R)) {materialFunction=/Engine/Functions/Engine_MaterialFunctions01/ImageAdjustment/LinearTosRGB}.sRGB) {name=Height L2, nodeColor=<rgba 0.4766 1 0 1>}
  n220 = MaterialFunctionCall() {materialFunction=&ObjectLocalBounds}
  n221 = Subtract(A=Subtract(A=WorldPosition() {worldPositionShaderOffset=WPT_ExcludeAllShaderOffsets}.XYZ, B=ActorPositionWS()), B=@n220.Local Bounds Minimum)
  n222 = NamedRerouteDeclaration(None=ComponentMask(None=Divide(A=@n221, B=@n220.Local Bounds Size)) {b=true}) {name=M_Bottom Gradient, nodeColor=<rgba 0.3828 0 1 1>}
  n223 = ComponentMask(None=VertexNormalWS()) {b=true}
  NamedRerouteDeclaration(None=Saturate(None=Max(A=@n222, B=@n223))) {name=Fade From Center To Top, nodeColor=<rgba 0 0.1562 1 1>}
  PinBase!(Output=@n106) {reroutePins [Output {expression=&self:MaterialExpressionReroute_98}]}
  PinBase!() {reroutePins [MaterialAttributes {expression=&self:MaterialExpressionReroute_70}]}
  PinBase() {reroutePins [
    MaterialAttributes {expression=&self:MaterialExpressionReroute_48}
    Snow Normals {expression=&self:MaterialExpressionReroute_49}
    L1 Normals {expression=&self:MaterialExpressionReroute_50}
    IsBaked {expression=&self:MaterialExpressionReroute_27}
   ]}
  PinBase!(Output=@n147) {reroutePins [Output {expression=&self:MaterialExpressionReroute_51}]}
  PinBase!() {reroutePins [MaterialAttributes {expression=&self:MaterialExpressionReroute_99}]}
  PinBase!(Result=@n111.Result) {reroutePins [Result {expression=&self:MaterialExpressionReroute_100}]}
  Reroute(None=@n112)
  SetMaterialAttributes!(MaterialAttributes, EmissiveColor=NamedRerouteUsage()) {attributeSetTypes [#9]}
  StaticBoolParameter() {parameterName="Add Top Masks (Pixel on Vertex)"}
  n224 = StaticSwitchParameter(True=NamedRerouteUsage(), False=Constant()) {group=D_L3 Custom Mask, parameterName=L3 Mask_Use VTX Colors, sortPriority=0}
  n225 = NamedRerouteUsage()
  n226 = StaticSwitchParameter(True=Max(A=@n224, B=@n225), False=Min(A=@n224, B=@n225)) {
   defaultValue=true
   desc=Additive/Subtractive Blending Mode switch
   group=D_L3 Custom Mask
   parameterName=L3 Mask_World Z Min Blend_Blend Max
   sortPriority=0
  }
  n227 = StaticSwitchParameter(True=@n226, False=@n224) {group=D_L3 Custom Mask, parameterName=L3 Mask_Use World Z Normal, sortPriority=0}
  n228 = Constant()
  n229 = StaticSwitchParameter(True=Max(A=@n227, B=@n228), False=Min(A=@n227, B=@n228)) {
   defaultValue=true
   desc=Additive/Subtractive Blending Mode switch
   group=D_L3 Custom Mask
   parameterName=L3 Mask_World XYZ Noise_Blend Max
   sortPriority=0
  }
  n230 = StaticSwitchParameter(True=@n229, False=@n227) {group=D_L3 Custom Mask, parameterName=L3 Mask_Use World WYZ Noise-WIP, sortPriority=0}
  n231 = NamedRerouteUsage()
  n232 = StaticSwitchParameter(True=Max(A=@n230, B=@n231), False=Min(A=@n230, B=@n231)) {
   defaultValue=true
   desc=Additive/Subtractive Blending Mode switch
   group=D_L3 Custom Mask
   parameterName=L3 Mask_AO_Blend Max
   sortPriority=0
  }
  n233 = StaticSwitchParameter(True=@n232, False=@n230) {group=D_L3 Custom Mask, parameterName=L3 Mask_Use AO-WIP, sortPriority=0}
  n234 = NamedRerouteUsage()
  n235 = ScalarParameter() {group=D_L3 Custom Mask, parameterName=L3 Mask_Bottom Gradient_Bias}
  n236 = Saturate(None=Multiply(A=Add(A=@n234, B=@n235), B=ScalarParameter() {group=D_L3 Custom Mask, parameterName=L3 Mask_Bottom Gradient_Scale}))
  n237 = StaticSwitchParameter(True=OneMinus(None=@n236), False=@n236) {group=D_L3 Custom Mask, parameterName=L3 Mask_Bottom Gradient_Invert}
  n238 = StaticSwitchParameter(True=Max(A=@n233, B=@n237), False=Min(A=@n233, B=@n237)) {
   defaultValue=true
   desc=Additive/Subtractive Blending Mode switch
   group=D_L3 Custom Mask
   parameterName=L3 Mask_Bottom Gradient_Blend Max
   sortPriority=0
  }
  n239 = StaticSwitchParameter(True=@n238, False=@n233) {group=D_L3 Custom Mask, parameterName=L3 Mask_Use Bottom Gradient-WIP, sortPriority=0}
  n240 = NamedRerouteUsage()
  n241 = StaticSwitchParameter(True=Max(A=@n239, B=@n240), False=Min(A=@n239, B=@n240)) {
   defaultValue=true
   desc=Additive/Subtractive Blending Mode switch
   group=D_L3 Custom Mask
   parameterName=L3 Mask_Waterline_Blend Max
   sortPriority=0
  }
  n242 = StaticSwitchParameter(True=@n241, False=@n239) {group=D_L3 Custom Mask, parameterName=L3 Mask_Use WaterLine-WIP, sortPriority=0}
  n243 = NamedRerouteUsage()
  n244 = StaticSwitchParameter(True=OneMinus(None=@n243), False=@n243) {group=D_L3 Custom Mask, parameterName=L3 Mask_SedimentLayers_Invert}
  n245 = StaticSwitchParameter(True=Max(A=@n242, B=@n244), False=Min(A=@n242, B=@n244)) {
   defaultValue=true
   desc=Additive/Subtractive Blending Mode switch
   group=D_L3 Custom Mask
   parameterName=L3 Mask_SedimentLayers_Blend Max
   sortPriority=0
  }
  n246 = StaticSwitchParameter(True=@n245, False=@n242) {group=D_L3 Custom Mask, parameterName=L3 Mask_Use SedimentLayers, sortPriority=0}
  n247 = Constant()
  n248 = StaticSwitchParameter(True=Max(A=@n246, B=@n247), False=Min(A=@n246, B=@n247)) {
   defaultValue=true
   desc=Additive/Subtractive Blending Mode switch
   group=D_L3 Custom Mask
   parameterName=L3 Mask_Height Blend_Blend Max
   sortPriority=0
  }
  StaticSwitchParameter(True=@n248, False=@n246) {group=D_L3 Custom Mask, parameterName=L3 Mask_Use Height Blend - WIP, sortPriority=0}
  n249 = StaticBoolParameter() {defaultValue=true, desc="On: L2 Use UV0\r\nOff: L2 Use UV1", group=0_Rock/Cliff Main Settings, parameterName=L2 Use Displacement}
  StaticSwitch(True=TextureCoordinate(), False=TextureCoordinate() {coordinateIndex=1}, Value=@n249)
 ]
}
