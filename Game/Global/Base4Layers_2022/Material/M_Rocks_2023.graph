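# Master rock material for the Base4Layers_2022 set, Default Lit with world-space normals and enabled for Nanite and instanced static meshes.
# MF_Rocks_2023 builds the material attributes, then a static switch overrides Opacity Mask with 1 or 0.
# The switch, on by default, is described as enabling vertex-color-driven grass spawning on the mesh.
# The attributes then pass through CINE_MF_EnvPostFX and MF_CreativeMoveTool.
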Material /Game/Global/Base4Layers_2022/Material/M_Rocks_2023
graph {
 properties {
  bTangentSpaceNormal=false
  bUseMaterialAttributes=true
  bUsedWithInstancedStaticMeshes=true
  bUsedWithNanite=true
  pixelDepthOffsetMode=PDOM_Legacy
  usedShadingModels=Default Lit
 }
 nodes [
  n1 = MaterialFunctionCall() {materialFunction=/Game/Global/Base4Layers_2022/MaterialFunctions/MF_Rocks_2023}
  n2 = StaticSwitchParameter(True=Constant() {r=1}, False=Constant()) {
   defaultValue=true
   desc="Set to true to spawn grass on this mesh.\r\nUses the vertex color to determine where to spawn."
   group=GrassSpawner
   parameterName="Use Grass(true) or a BaseTilingTexture(False)"
   sortPriority=0
  }
  MaterialFunctionCall(MA In=MaterialFunctionCall!(In=SetMaterialAttributes!(MaterialAttributes=@n1.Result, Opacity Mask=@n2) {attributeSetTypes [#1]}) {materialFunction=/Game/Packages/Fortress_SharedMaterials/Cinematics/CINE_MF_EnvPostFX}.Result) {materialFunction=&MF_CreativeMoveTool}
 ]
}
