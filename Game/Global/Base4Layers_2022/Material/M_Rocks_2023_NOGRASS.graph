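# M_Rocks_2023_NOGRASS is a Default Lit rock material, set up for instanced static meshes and Nanite.
# It uses material attributes with world-space normals, taking its whole surface from the MF_Rocks_2023 function.
# That result feeds CINE_MF_EnvPostFX, a shared cinematics function that is the last node in the graph.
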Material /Game/Global/Base4Layers_2022/Material/M_Rocks_2023_NOGRASS
graph {
 properties {
  bTangentSpaceNormal=false
  bUseMaterialAttributes=true
  bUsedWithInstancedStaticMeshes=true
  bUsedWithNanite=true
  pixelDepthOffsetMode=PDOM_Legacy
  usedShadingModels=Default Lit
 }
 nodes [MaterialFunctionCall!(In=MaterialFunctionCall() {materialFunction=/Game/Global/Base4Layers_2022/MaterialFunctions/MF_Rocks_2023}.Result) {materialFunction=/Game/Packages/Fortress_SharedMaterials/Cinematics/CINE_MF_EnvPostFX}]
}
