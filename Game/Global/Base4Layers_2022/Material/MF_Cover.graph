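# Material function that layers an optional cover, such as moss or Helios landscape snow, over input material attributes.
# Switchable masks (baked, height gradient, noise, slope, directional, DFAO, spec texture, per-layer) combine with cover texture height to drive the blend.
# The pre-height mask also drives border dirt with optional leaks, skipped on mobile and low quality.
# Textures use top projection or dithered single-sample triplanar in world or pre-skinned local space; a debug switch displays masks.
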MaterialFunction /Game/Global/Base4Layers_2022/Material/MF_Cover
graph {
 properties {description=Preskined LS UV}
 nodes [
  n1 = NamedRerouteUsage()
  n2 = StaticSwitchParameter(True=VertexColor(), False=NamedRerouteUsage()) {defaultValue=true, group=Cover, parameterName="T= Vertex Color F=SFX texture", sortPriority=300}
  n3 = Multiply(
   A=Add(
    A=StaticComponentMaskParameter(None=@n2) {defaultR=true, group=Cover, parameterName=Cover_Baked_Channel, sortPriority=315}
    B=ScalarParameter() {group=Cover, parameterName="Mask_Baked_Bias (Contract/Expand)", sortPriority=315}
   )
   B=ScalarParameter() {defaultValue=1, group=Cover, parameterName="Mask_Baked_Scale (Hardness)", sortPriority=315}
  )
  n4 = ScalarParameter() {group=Cover, parameterName=Cover_Flat Amount, sliderMax=1, sortPriority=0}
  n5 = NamedRerouteUsage()
  n6 = Constant()
  n7 = NamedRerouteUsage()
  n8 = Add(
   A=StaticSwitchParameter(True=Add(A=@n4, B=@n3), False=@n4) {group=Cover, parameterName="== Mask_Baked Mask(VTX or T_SFX) ==", sortPriority=300}
   B=Max(
    A=StaticSwitchParameter(True=@n5, False=@n6) {defaultValue=true, group=Cover, parameterName="== Mask_Gradient TOP==", sortPriority=400}
    B=StaticSwitchParameter(True=@n7, False=@n6) {defaultValue=true, group=Cover, parameterName="== Mask_Gradient BOTTOM==", sortPriority=410}
   )
  )
  n9 = StaticSwitchParameter(True=Add(A=@n8, B=NamedRerouteUsage()), False=@n8) {group=Cover, parameterName="== Mask_WS noise ==", sortPriority=550}
  n10 = StaticSwitchParameter(True=Add(A=@n9, B=NamedRerouteUsage()), False=@n9) {group=Cover, parameterName="== Mask_Sides&&Top_ ==", sortPriority=500}
  Add(A=@n10, B=@n1)
  Clamp() {minDefault=-1}
  n11 = NamedRerouteUsage()
  n12 = Multiply(A=@n11, B=Constant4Vector() {constant=<rgba 1 0 0 1>}.RGB)
  n13 = Constant()
  n14 = SetMaterialAttributes!(
   MaterialAttributes
   EmissiveColor=LinearInterpolate(A=StaticSwitchParameter(True=@n12, False=@n11) {group=Cover, parameterName=Border Dirt, sortPriority=13}, Alpha=NamedRerouteUsage())
   OpacityMask=NamedRerouteUsage()
   Roughness=Constant() {r=1}
   Specular=@n13
   Metallic=@n13
  ) {attributeSetTypes [
    #1
    #2
    #3
    #4
    #5
   ]}
  n15 = NamedRerouteUsage()
  n16 = GetMaterialAttributes(None=FunctionInput(Preview=MakeMaterialAttributes!(
     BaseColor
     Metallic
     Specular
     Roughness
     Anisotropy
     EmissiveColor
     Opacity
     OpacityMask
     Normal
     Tangent
     WorldPositionOffset
     SubsurfaceColor
     ClearCoat
     ClearCoatRoughness
     AmbientOcclusion
     Refraction
     CustomizedUVs_0
     CustomizedUVs_1
     CustomizedUVs_2
     CustomizedUVs_3
     CustomizedUVs_4
     CustomizedUVs_5
     CustomizedUVs_6
     CustomizedUVs_7
     PixelDepthOffset
     ShadingModel
     Displacement
    )) {bUsePreviewValueAsDefault=true, inputType=FunctionInput_MaterialAttributes}) {attributeGetTypes [
    #6
    #3
    #4
    #5
    #7
    #2
   ]}
  n17 = SetMaterialAttributes!(
   MaterialAttributes=@n16.MaterialAttributes
   BaseColor
   Metallic=Multiply(A=NamedRerouteUsage(), B=ScalarParameter() {defaultValue=1, group=Cover, parameterName=Base_Metallic_Mult, sortPriority=62})
   Specular
   Roughness
   Normal
  ) {attributeSetTypes [
    #6
    #5
    #4
    #3
    #7
   ]}
  n18 = SetMaterialAttributes(
   MaterialAttributes=@n17
   BaseColor=NamedRerouteUsage()
   Specular=NamedRerouteUsage()
   Roughness=NamedRerouteUsage()
   Normal=NamedRerouteUsage()
  ) {attributeSetTypes [
    #6
    #5
    #4
    #3
    #7
   ]}
  n19 = BlendMaterialAttributes(A=@n17, B=@n18, Alpha=NamedRerouteUsage()) {vertexAttributeBlendType=UseB}
  n20 = StaticSwitchParameter(True=ShadingPathSwitch(Default=QualitySwitch(Default=@n19, Low=@n17), Mobile=@n17), False=@n17) {group=Cover, parameterName=Border Dirt, sortPriority=13}
  n21 = SetMaterialAttributes(
   MaterialAttributes=@n20
   BaseColor=NamedRerouteUsage()
   Metallic=Constant()
   Specular=NamedRerouteUsage()
   Roughness=NamedRerouteUsage()
   Normal=NamedRerouteUsage()
  ) {attributeSetTypes [
    #6
    #5
    #4
    #3
    #7
   ]}
  n22 = StaticBoolParameter() {desc="Only supported with up to 2 layers!", group=Cover, parameterName=Use Landscape Snow, sortPriority=3}
  n23 = SetMaterialAttributes!(
   MaterialAttributes=BlendMaterialAttributes(A=@n20, B=StaticSwitch(True=@n15, False=@n21, Value=@n22), Alpha=NamedRerouteUsage()) {vertexAttributeBlendType=UseB}
   OpacityMask=NamedRerouteUsage()
  ) {attributeSetTypes [#2]}
  FunctionOutput(None=StaticSwitchParameter(
    True=StaticSwitchParameter(True=@n14, False=@n23) {group=Cover, parameterName="[DEBUG] Cover Mask", sortPriority=-9}
    False=@n16.MaterialAttributes
   ) {group=Cover, parameterName=Use Cover, sortPriority=-10})
  LinearInterpolate()
  LinearInterpolate(
   A=Subtract(A=WorldPosition().XYZ, B=ActorPositionWS())
   Alpha=ScalarParameter() {defaultValue=0.35, group=Cover, parameterName="Mask_Gradient_0=WS 1=Local", sortPriority=400}
  )
  LocalPosition() {localOrigin=InstancePreSkinning}
  n24 = ScalarParameter() {group=Cover, parameterName=Border Dirt_LeakTriplanarBiasX, sliderMax=1, sliderMin=-1, sortPriority=15}
  n25 = ScalarParameter() {group=Cover, parameterName=Border Dirt_LeakTriplanarBiasY, sliderMax=1, sliderMin=-1, sortPriority=15}
  n26 = ScalarParameter() {group=Cover, parameterName=Border Dirt_LeakTriplanarBiasZ, sliderMax=1, sliderMin=-1, sortPriority=15}
  n27 = Abs(None=Add(
    A=VertexNormalWS()
    B=MaterialFunctionCall(R=@n24, G=@n25, B=@n26) {materialFunction=&AppendMany}.RGB
   ))
  n28 = ComponentMask(None=@n27) {g=true, r=true}
  n29 = ComponentMask(None=@n27) {b=true}
  n30 = Saturate(None=Subtract(A=@n28, B=@n29))
  n31 = ComponentMask(None=@n30) {r=true}
  n32 = ComponentMask(None=@n30) {g=true}
  n33 = Subtract(A=@n31, B=@n32)
  n34 = NamedRerouteDeclaration(None=Ceil(None=@n33)) {name=Flat R_1, nodeColor=<rgba 1 0 0.8906 1>}
  n35 = NamedRerouteDeclaration(None=Ceil(None=Subtract(A=@n32, B=@n31))) {name=Flat G_1, nodeColor=<rgba 0 1 0.7891 1>}
  n36 = Max(A=@n34, B=@n35)
  MaterialFunctionCall!(
   Vector3=AppendVector(A=AppendVector(A=@n34, B=@n35), B=OneMinus(None=@n36))
   MaximumNumberOfDigits
   UVs
   DebugTextLocation RG_UpperRight BA_LowerLeft
   Component Spacing
  ) {materialFunction=/Engine/Functions/Engine_MaterialFunctions02/Utility/DebugFloat3Values}
  MaterialFunctionCall()
  Multiply(A=NamedRerouteUsage(), B=ScalarParameter() {defaultValue=1, group=Cover, parameterName=Base_Specular_Mult, sortPriority=62})
  Multiply(A=NamedRerouteUsage(), B=ScalarParameter() {defaultValue=1, group=Cover, parameterName=Base_Roughness_Mult, sortPriority=62})
  n37 = ComponentMask(None=LocalPosition().XYZ) {g=true}
  n38 = ScalarParameter() {
   defaultValue=1
   desc="Mask based on mesh local Y. Will make what is on negative side from pivot free of dirt effect. In some cases you can't use it ( Curved walls) so instead use additional material layer to controll dirt level inside buildings"
   group=Cover
   parameterName=Architecture_Interior Wall Cover Value
   sortPriority=0
  }
  n39 = Max(A=Saturate(None=Multiply(A=@n37) {constB=0.2}), B=@n38)
  n40 = ShaderStageSwitch(PixelShader=VertexInterpolator(VS=@n39).PS, VertexShader=@n39)
  n41 = Max(A=@n40, B=ScalarParameter() {defaultValue=1, group=Cover, parameterName=Border Dirt_LeaksInteriorWall, sliderMax=1, sortPriority=15})
  n42 = ScalarParameter() {defaultValue=1, group=Cover, parameterName=Border Dirt_LeaksI4Layer, sliderMax=1, sortPriority=15}
  n43 = Saturate(None=MaterialFunctionCall!(
    1=ScalarParameter() {defaultValue=1, group=Cover, parameterName=Cover Value Layer 1, sortPriority=900}
    2=ScalarParameter() {defaultValue=1, group=Cover, parameterName=CoverValue Layer 2, sortPriority=900}
    3=ScalarParameter() {defaultValue=1, group=Cover, parameterName=CoverValue Layer 3, sortPriority=900}
    4=ScalarParameter() {defaultValue=1, group=Cover, parameterName=Cover Value Layer 4, sortPriority=900}
   ) {materialFunction=/Game/Global/Base4Layers_2022/MaterialFunctions/MF_UV2LerpScalar_Layers_New}.Result)
  n44 = Divide(
   A=Multiply(A=WorldPosition().XYZ, B=Constant3Vector() {constant=<rgba 1 1 -1 0>})
   B=ScalarParameter() {defaultValue=2000, group=Cover, parameterName=Border Dirt_LeakScale, sortPriority=15}
  )
  n45 = ComponentMask(None=@n44) {g=true, r=true}
  n46 = ComponentMask(None=@n44) {b=true, r=true}
  n47 = NamedRerouteUsage()
  n48 = TextureSample(UVs=LinearInterpolate(A=LinearInterpolate(A=@n45, B=@n46, Alpha=@n47), B=ComponentMask(None=@n44) {b=true, g=true}, Alpha=NamedRerouteUsage())) {texture=/Game/Global/Base4Layers_2022/Textures/T_Leaks}
  n49 = Subtract(A=@n29, B=@n28)
  n50 = Min(A=ComponentMask(None=@n49) {r=true}, B=ComponentMask(None=@n49) {g=true})
  n51 = OneMinus(None=Max(A=Saturate(None=@n50), B=Abs(None=@n33)))
  n52 = Constant() {r=64}
  n53 = StaticSwitchParameter(
   True=LinearInterpolate(
    A=@n48.R
    B=ScalarParameter() {defaultValue=1, group=Cover, parameterName=Border Dirt_LeaksSeamFixValue, sortPriority=15}
    Alpha=Power(Base=@n51, Exp=@n52)
   )
   False=@n48.R
  ) {group=Cover, parameterName=Border Dirt_LeaksSeamFix, sortPriority=15}
  n54 = ScalarParameter() {group=Cover, parameterName=Border Dirt_LeaksSides, sortPriority=15}
  n55 = ScalarParameter() {group=Cover, parameterName=Border Dirt_LeaksTop, sortPriority=15}
  n56 = VertexNormalWS()
  n57 = LinearInterpolate(
   A=ComponentMask(None=Multiply(A=@n53, B=@n54)) {r=true}
   B=ComponentMask(None=Multiply(A=@n53, B=@n55)) {g=true}
   Alpha=ComponentMask(None=Abs(None=@n56)) {b=true}
  )
  n58 = StaticSwitchParameter(True=Multiply(A=@n57, B=Max(A=@n43, B=@n42)), False=@n57) {group=Cover, parameterName="==Mask_4 Layer==", sortPriority=900}
  n59 = NamedRerouteUsage()
  n60 = NamedRerouteUsage()
  n61 = NamedRerouteUsage()
  n62 = Add(
   A=StaticComponentMaskParameter(None=MaterialFunctionCall(R=@n59, G=@n60, B=@n61) {materialFunction=&AppendMany}.RGB) {defaultR=true, group=Cover, parameterName=Spec Tex_Channel, sortPriority=61}
   B=ScalarParameter() {defaultValue=-0.52, group=Cover, parameterName="SRM_Bias (Contract/Expand)", sortPriority=61}
  )
  n63 = ScalarParameter() {defaultValue=-5, group=Cover, parameterName="SRM_Mult (Hardness)", sortPriority=61}
  n64 = DistanceFieldApproxAO!(
   World Position
   Normal
   BaseDistance=ScalarParameter() {defaultValue=10, group=Cover, parameterName=DFAO_Distance, sortPriority=51}
   Radius=ScalarParameter() {defaultValue=300, group=Cover, parameterName=DFAO_Radius, sortPriority=51}
  ) {numSteps=3}
  n65 = Divide(A=Subtract(A=PixelDepth().R, B=Constant() {r=5000}), B=Constant() {r=10000})
  n66 = Saturate(None=Subtract(A=OneMinus(None=@n64), B=Saturate(None=@n65)))
  n67 = ScalarParameter() {defaultValue=1, group=Cover, parameterName=DFAO_Power, sortPriority=51}
  n68 = Constant()
  n69 = QualitySwitch(
   Default=@n68
   Epic=Multiply(A=Power(Base=@n66, Exp=@n67), B=ScalarParameter() {defaultValue=1, group=Cover, parameterName="DFAO_Mult (hardness)", sortPriority=51})
  )
  n70 = StaticSwitchParameter(True=Subtract(A=@n10, B=@n1), False=@n10) {
   desc=Uses the Direction Vector to add Substance on one side and subtract from the other. EX: moss that wont grow as much on the sunny side and develops better on the dark side
   group=Cover
   parameterName="== Mask_WS Directional Subtract=="
   sortPriority=510
  }
  n71 = StaticSwitchParameter(True=Add(A=@n70, B=ShadingPathSwitch(Default=@n69, Mobile=@n68)), False=@n70) {
   desc="DFAO is Only supported on High end Platforms (epic Settings)\r\nUse baked Ao if possible for compability"
   group=Cover
   parameterName="== Mask_DFAO (Epic Only) =="
   sortPriority=50
  }
  n72 = StaticSwitchParameter(True=Add(A=@n71, B=Multiply(A=@n62, B=@n63)), False=@n71) {group=Cover, parameterName="== Mask_SMR Texture ==", sortPriority=60}
  n73 = Multiply(A=StaticSwitchParameter(True=Multiply(A=@n72, B=@n43), False=@n72) {group=Cover, parameterName="==Mask_4 Layer==", sortPriority=900}, B=@n40)
  n74 = StaticSwitchParameter(True=Add(A=@n73, B=Multiply(A=@n41, B=@n58)), False=@n73) {group=Cover, parameterName=Border Dirt_Leaks, sortPriority=15}
  n75 = MaterialFunctionCall(
   In=Add(A=@n74, B=ScalarParameter() {defaultValue=0.2, group=Cover, parameterName=Border Dirt_SIDES_Radius, sortPriority=14})
   Contrast=ScalarParameter() {defaultValue=2, group=Cover, parameterName=Border Dirt_SIDES_Contrast, sortPriority=14}
  ) {materialFunction=&CheapContrast}
  n76 = MaterialFunctionCall(
   In=Add(A=@n74, B=ScalarParameter() {defaultValue=0.2, group=Cover, parameterName=Border Dirt_TOP_Radius, sortPriority=14})
   Contrast=ScalarParameter() {defaultValue=2, group=Cover, parameterName=Border Dirt_TOP_Contrast, sortPriority=14}
  ) {materialFunction=&CheapContrast}
  n77 = Add(
   A=ComponentMask(None=VertexNormalWS()) {b=true}
   B=ScalarParameter() {group=Cover, parameterName=-Border Dirt_T/S_ExpandTop, sortPriority=14}
  )
  n78 = ScalarParameter() {group=Cover, parameterName=-Border Dirt_T/S_Contrast, sortPriority=14}
  n79 = NamedRerouteDeclaration(None=Saturate(None=MaterialFunctionCall(In=@n77, Contrast=@n78) {materialFunction=&CheapContrast}.Result)) {name=BorderDirt_TopMask, nodeColor=<rgba 1 0 0.6797 1>}
  n80 = Add(A=@n74, B=ScalarParameter() {defaultValue=0.2, group=Cover, parameterName=Border Dirt_Radius, sortPriority=14})
  n81 = ScalarParameter() {defaultValue=2, group=Cover, parameterName=Border Dirt_Contrast, sortPriority=14}
  n82 = StaticBoolParameter() {group=Cover, parameterName=Border Dirt_Top/Sides, sortPriority=13}
  NamedRerouteDeclaration(None=StaticSwitch(
    True=LinearInterpolate(A=@n75.Result, B=@n76.Result, Alpha=@n79)
    False=MaterialFunctionCall(In=@n80, Contrast=@n81) {materialFunction=&CheapContrast}.Result
    Value=NamedRerouteDeclaration(None=@n82) {name=UseSide/TopMA, nodeColor=<rgba 0 1 1 1>}
   )) {name=Mask_Border Dirt, nodeColor=<rgba 0.9219 1 0 1>}
  NamedRerouteDeclaration(None=@n16.BaseColor) {name=BaseColor_IN, nodeColor=<rgba 1 0.8672 0 1>}
  NamedRerouteDeclaration(None=@n16.Specular) {name=Specular_IN, nodeColor=<rgba 0 1 0.5781 1>}
  NamedRerouteDeclaration(None=@n16.Metallic) {name=Metallic_IN, nodeColor=<rgba 0.7109 1 0 1>}
  n83 = Saturate(None=@n73)
  n84 = LinearInterpolate(
   A=ScalarParameter() {group=Cover, parameterName=Cover_Height_Min, sortPriority=5}
   B=ScalarParameter() {defaultValue=1, group=Cover, parameterName=Cover_Height_Max, sortPriority=5}
   Alpha=NamedRerouteUsage()
  )
  n85 = Add(A=Multiply(A=@n83) {constB=2, desc=2}, B=Subtract(A=Saturate(None=@n84)) {desc=1})
  NamedRerouteDeclaration(None=StaticSwitch(
    True=NamedRerouteUsage() {desc="Only supported with up to 2 layers!"}
    False=Saturate(None=@n85)
    Value=StaticBoolParameter() {desc="Only supported with up to 2 layers!", group=Cover, parameterName=Use Landscape Snow, sortPriority=3}
   )) {name=Cover Mask, nodeColor=<rgba 1 0 0.3516 1>}
  n86 = ScalarParameter() {
   desc="0= Use Texture Normals\r\n1= Use Mesh Vertex Normals"
   group=Cover
   parameterName="Mask_Sides&&Top_Normals (0=Tex 1=Vtx)"
   sortPriority=500
  }
  n87 = LinearInterpolate(A=Transform(None=NamedRerouteUsage()), B=VertexNormalWS(), Alpha=@n86)
  n88 = Abs(None=ComponentMask(None=@n87) {g=true, r=true})
  n89 = NamedRerouteDeclaration(None=ComponentMask(None=@n87) {b=true}) {name=N_Z, nodeColor=<rgba 0.5469 1 0 1>}
  n90 = ScalarParameter() {group=Cover, parameterName=Mask_Sides, sortPriority=505}
  n91 = ScalarParameter() {defaultValue=1, group=Cover, parameterName=Mask_Top/Bottom Bias, sortPriority=505}
  n92 = Clamp(None=DotProduct(
    A=AppendVector(A=@n88, B=@n89)
    B=MaterialFunctionCall(R=@n90, G=@n90, B=@n91) {materialFunction=&AppendMany}.RGB
   )) {minDefault=-1}
  n93 = ScalarParameter() {defaultValue=1, group=Cover, parameterName="Mask_Sides&&top_Bias (Contract/Expand)", sortPriority=506}
  NamedRerouteDeclaration(None=Multiply(A=Add(A=@n92, B=@n93), B=ScalarParameter() {defaultValue=1, group=Cover, parameterName="Mask_Sides&&Top_Mult (Hardness)", sortPriority=506})) {name="M_Sides&&Top", nodeColor=<rgba 1 0.6328 0 1>}
  NamedRerouteDeclaration(None=@n16.Normal) {name=Normal_IN, nodeColor=<rgba 0 0.5312 1 1>}
  n94 = Add(A=WorldPosition().Z, B=ScalarParameter() {group=Cover, parameterName=WS TOP/BOTTOM Gradients_Offset, sortPriority=399})
  n95 = ScalarParameter() {defaultValue=384, group=Cover, parameterName=WS TOP/BOTTOM Gradients_Tile, sortPriority=399}
  n96 = MaterialFunctionCall() {materialFunction=&ObjectLocalBounds}
  n97 = Divide(A=Subtract(A=LocalPosition() {localOrigin=InstancePreSkinning}.XYZ, B=@n96.Local Bounds Minimum), B=@n96.Local Bounds Size)
  n98 = StaticSwitchParameter(True=Frac(None=Divide(A=@n94, B=@n95)), False=VertexInterpolator(VS=ComponentMask(None=@n97) {b=true}).PS) {group=Cover, parameterName="** MASK WS TOP/BOTTOM Gradients (Architecture) **", sortPriority=399}
  n99 = Divide(
   A=Add(
    A=Subtract(A=@n98)
    B=ScalarParameter() {defaultValue=0.35, group=Cover, parameterName="Mask_Gradient Top_Bias (Contract/Expand)", sortPriority=405}
   )
   B=ScalarParameter() {defaultValue=1, group=Cover, parameterName="Mask_Gradient Top_Mult (Softness)", sortPriority=405}
  )
  NamedRerouteDeclaration(None=Power(
    Base=Saturate(None=@n99)
    Exp=ScalarParameter() {defaultValue=1, group=Cover, parameterName=Mask_Gradient Top_Power, sliderMax=32, sortPriority=405}
   )) {name=M_Z Gradient TOP, nodeColor=<rgba 1 0.2578 0 1>}
  n100 = Add(A=@n98)
  n101 = Divide(
   A=Add(
    A=OneMinus(None=@n100)
    B=ScalarParameter() {defaultValue=0.35, group=Cover, parameterName="Mask_Gradient Bottom_Bias (Contract/Expand)", sortPriority=425}
   )
   B=ScalarParameter() {defaultValue=0.5, group=Cover, parameterName="Mask_Gradient Bottom_Mult (Softness)", sortPriority=425}
  )
  NamedRerouteDeclaration(None=Power(
    Base=Saturate(None=@n101)
    Exp=ScalarParameter() {defaultValue=1, group=Cover, parameterName=Mask_Gradient Bottom_Power, sliderMax=32, sortPriority=425}
   )) {name=M_Z Gradient BOTTOM, nodeColor=<rgba 1 0.2578 0 1>}
  n102 = VectorParameter() {
   desc="Compare an arbitraty vector with the mesh world normals. Matching values are masked out. \r\nUsefull for global behaviours such as removing moss or snow coverage from the sunny side of multiple assets that share a aprent material."
   group=Cover
   parameterName=Mask_Directional_Vector
   sortPriority=520
   channelNames {a=--, b=Z, g=Y, r=X}
   defaultValue=<rgba 0.5 -0.8 1 1>
  }
  n103 = NamedRerouteUsage()
  n104 = Add(
   A=DotProduct(A=@n102, B=Transform(None=@n103))
   B=ScalarParameter() {defaultValue=-0.2, group=Cover, parameterName="Mask_Directional_Bias (Contract/Expand)", sortPriority=521}
  )
  n105 = ScalarParameter() {defaultValue=0.3, group=Cover, parameterName="Mask_Directional_Mult (Hardness)", sortPriority=521}
  NamedRerouteDeclaration(None=Saturate(None=Multiply(A=@n104, B=@n105))) {name=M_WS Directional, nodeColor=<rgba 0.2188 1 0 1>}
  n106 = NamedRerouteUsage()
  n107 = VectorParameter() {
   desc="Alpha 0=Multiply\r\nAlpha 1=Solid Color Lerp"
   group=Cover
   parameterName=Border Dirt_Color
   sortPriority=15
   channelNames {a="A=0 (Mult)  A=1(Solid Color)", b="", g="", r=""}
   defaultValue=<rgba 0.4149 0.2605 0.1379 1>
  }
  n108 = LinearInterpolate(A=Multiply(A=@n107) {constB=0.15}, B=@n107, Alpha=NamedRerouteUsage())
  NamedRerouteDeclaration(None=LinearInterpolate(A=Multiply(A=@n106, B=@n108), B=@n108, Alpha=@n107."A=0 (Mult)  A=1(Solid Color)")) {name=BorderDirt_Base Color, nodeColor=<rgba 0.5469 0 1 1>}
  n109 = LocalPosition() {localOrigin=InstancePreSkinning}
  n110 = ScalarParameter() {defaultValue=500, group=Cover, parameterName=Cover_Tile, sortPriority=5}
  n111 = Divide(
   A=StaticSwitch(
    True=VertexInterpolator(VS=@n109.XY).PS
    False=WorldPosition() {worldPositionShaderOffset=WPT_ExcludeAllShaderOffsets}.XY
    Value=NamedRerouteUsage()
   )
   B=@n110
  ) {constB=120}
  n112 = TextureObjectParameter() {group=Cover, parameterName=Cover Diffuse, sortPriority=5, texture=/Game/Global/Textures/Default/Blanks/T_FlatGrey}
  n113 = ScalarParameter() {defaultValue=20, group=Cover, parameterName=Cover_Triplanar Fade, sortPriority=5}
  n114 = MaterialFunctionCall!("Normal (WS)", Texture Scale=@n110, Triplanar Fade=@n113, Derivatives Scale) {materialFunction=/Game/Athena/Environments/MaterialFunctions/MF_TriplanarDithered}
  n115 = NamedRerouteUsage()
  n116 = StaticSwitch(True=NamedRerouteUsage(), False=@n114.UVs, Value=@n115)
  n117 = StaticSwitch(
   True=TextureSample(UVs=@n111, Tex=@n112) {texture=&T_FlatGrey}.RGBA
   False=TextureSample!(UVs=@n116, Tex=@n112, "DDX(UVs)"=NamedRerouteUsage(), "DDY(UVs)"=NamedRerouteUsage(), Apply View MipBias) {mipValueMode=TMVM_Derivative, texture=&T_FlatGrey}.RGBA
   Value=NamedRerouteUsage()
  )
  n118 = StaticBoolParameter() {defaultValue=true, group=Cover, parameterName="Use WS Texture for Color && Specular", sortPriority=1}
  NamedRerouteDeclaration(None=StaticSwitch(True=ComponentMask(None=@n117) {a=true}, False=Constant() {r=1}, Value=@n118)) {name=Cover_H, nodeColor=<rgba 1 0 0.375 1>}
  n119 = VectorParameter() {
   desc="Alpha 0=Multiply\r\nAlpha 1=Solid Color Lerp"
   group=Cover
   parameterName=Cover_Color
   sortPriority=5
   defaultValue=<rgba 0.7531 0.7531 0.7531 1>
  }
  n120 = Multiply(A=NamedRerouteUsage(), B=@n119.RGB)
  NamedRerouteDeclaration(None=StaticSwitch(True=ComponentMask(None=@n117) {b=true, g=true, r=true}, False=LinearInterpolate(A=@n120, B=@n119.RGB, Alpha=@n119.RGB), Value=@n118)) {name=Cover_Base Color, nodeColor=<rgba 0 0.6953 1 1>}
  n121 = TextureObjectParameter() {
   group=Cover
   parameterName=Cover Normals
   samplerType=SAMPLERTYPE_Normal
   sortPriority=10
   texture=&FlatNormalBlue
  }
  n122 = TextureSample!(UVs=@n111, Tex=@n121, Apply View MipBias) {samplerType=SAMPLERTYPE_Normal, texture=&FlatNormalBlue}
  n123 = NamedRerouteUsage()
  n124 = NamedRerouteUsage()
  n125 = NamedRerouteUsage()
  n126 = Transform(None=MaterialFunctionCall!(
    Tangent Normals=TextureSample!(UVs=@n116, Tex=@n121, "DDX(UVs)"=@n123, "DDY(UVs)"=@n124, Apply View MipBias) {mipValueMode=TMVM_Derivative, samplerType=SAMPLERTYPE_Normal, texture=&FlatNormalBlue}.RGB
    Vertex Normals
    Triplanar Mask=StaticSwitch(True=@n125, False=@n114.Mask, Value=@n115)
   ) {materialFunction=/Game/Athena/Environments/MaterialFunctions/MF_TriplanarNormalTangentToWorld}.World Normals) {transformSourceType=TRANSFORMSOURCE_World, transformType=TRANSFORM_Tangent}
  n127 = NamedRerouteUsage()
  n128 = ScalarParameter() {group=Cover, parameterName=Cover Normals_SIDES_Flat, sortPriority=10}
  n129 = ScalarParameter() {group=Cover, parameterName=Cover Normals_TOP_Flat, sortPriority=10}
  n130 = NamedRerouteUsage()
  NamedRerouteDeclaration(None=MaterialFunctionCall(Normal=StaticSwitch(True=@n122.RGB, False=@n126, Value=@n127), Flatness=LinearInterpolate(A=@n128, B=@n129, Alpha=@n130)) {materialFunction=&FlattenNormal}.Result) {name=Cover Normal, nodeColor=<rgba 1 0.8672 0 1>}
  n131 = TextureObjectParameter() {
   group=Cover
   parameterName=Cover Specular
   samplerType=SAMPLERTYPE_Masks
   sortPriority=7
   texture=&T_Fortnite_Default_S
  }
  n132 = NamedRerouteUsage()
  n133 = NamedRerouteUsage()
  n134 = MaterialFunctionCall(Color=StaticSwitch(
    True=TextureSample!(UVs=@n111, Tex=@n131, Apply View MipBias) {samplerType=SAMPLERTYPE_Masks, texture=&T_Fortnite_Default_S}.RGB
    False=TextureSample!(UVs=@n116, Tex=@n131, "DDX(UVs)"=@n132, "DDY(UVs)"=@n133, Apply View MipBias) {mipValueMode=TMVM_Derivative, samplerType=SAMPLERTYPE_Masks, texture=&T_Fortnite_Default_S}.RGB
    Value=NamedRerouteUsage()
   )) {materialFunction=&SplitComponents}
  NamedRerouteDeclaration(None=StaticSwitch(True=@n134.R, False=ScalarParameter() {defaultValue=0.5, group=Cover, parameterName=Cover_Specular, sortPriority=8}, Value=@n118)) {name=Cover_Specular, nodeColor=<rgba 0.5469 0 1 1>}
  NamedRerouteDeclaration(None=StaticSwitch(True=@n134.B, False=ScalarParameter() {defaultValue=0.85, group=Cover, parameterName=Cover_Roughness, sortPriority=8}, Value=@n118)) {name=Cover_Roughness, nodeColor=<rgba 0 1 0.2266 1>}
  n135 = VertexInterpolator(VS=LocalPosition() {localOrigin=InstancePreSkinning}.XYZ)
  n136 = Constant3Vector() {constant=<rgba 1 1 -1 0>}
  n137 = Divide(A=NamedRerouteDeclaration(None=Multiply(A=@n135.PS, B=@n136)) {name=PreskinLocal_1, nodeColor=<rgba 0.3359 1 0 1>}, B=@n110) {constB=120}
  n138 = Abs(None=Transform(None=VertexNormalWS()) {transformSourceType=TRANSFORMSOURCE_World, transformType=TRANSFORM_Local})
  n139 = Power(Base=@n138, Exp=@n113) {constExponent=10}
  n140 = MaterialFunctionCall(Float3=Saturate(None=@n139)) {materialFunction=&BreakOutFloat3Components}
  n141 = Divide(A=@n139, B=Add(A=Add(A=@n140.R, B=@n140.R), B=@n140.B))
  n142 = ViewProperty() {property=MEVP_TemporalSampleIndex}
  MaterialExpressionAdd_17 = Add(A=ScreenPosition().PixelPosition, B=AppendVector(A=@n142.Property, B=@n142.Property))
  n143 = Custom!(p=@MaterialExpressionAdd_17) {
   code="Mod( ((uint)(p.x) + 2 * (uint)(p.y)) , 5 )"
   outputType=CMOT_Float1
   inputs [{inputName=p, input {expression=&self:MaterialExpressionAdd_17}}]
  }
  n144 = TextureSample!(
   UVs=MaterialFunctionCall!(TextureResolution=Constant2Vector() {g=64, r=64}) {materialFunction=/Engine/Functions/Engine_MaterialFunctions02/Texturing/ScreenAlignedPixelToPixelUVs}.Result
   Tex
   Apply View MipBias
  ) {samplerType=SAMPLERTYPE_LinearGrayscale, texture=/Engine/EngineMaterials/Good64x64TilingNoiseHighFreq}
  n145 = Max(A=Multiply(A=Add(A=@n143, B=@n144.R) {constB=0}) {constB=0.1667}) {constB=0.001}
  n146 = AppendVector(A=@n145, B=OneMinus(None=@n145))
  n147 = Floor(None=Add(A=ComponentMask(None=@n141) {g=true, r=true}, B=@n146))
  n148 = ComponentMask(None=@n147) {r=true}
  n149 = Multiply(A=ComponentMask(None=@n137) {b=true, g=true}, B=@n148)
  n150 = ComponentMask(None=@n147) {g=true}
  n151 = Multiply(A=ComponentMask(None=@n137) {b=true, r=true}, B=@n150)
  n152 = ComponentMask(None=@n137) {g=true, r=true}
  n153 = OneMinus(None=Max(A=@n148, B=@n150))
  NamedRerouteDeclaration(None=Add(A=Add(A=@n149, B=@n151), B=Multiply(A=@n152, B=@n153))) {name=Cover_Tripl_LocalUV, nodeColor=<rgba 1 0 0.6562 1>}
  NamedRerouteDeclaration(None=AppendVector(A=AppendVector(A=@n148, B=@n150), B=@n153)) {name=Cover_Tripl_LocalMask, nodeColor=<rgba 1 0 0.6562 1>}
  n154 = Divide(A=NamedRerouteUsage(), B=NamedRerouteUsage()) {constB=120}
  n155 = Power(Base=@n138, Exp=NamedRerouteUsage()) {constExponent=10}
  n156 = MaterialFunctionCall(Float3=Saturate(None=@n155)) {materialFunction=&BreakOutFloat3Components}
  n157 = Divide(A=@n155, B=Add(A=Add(A=@n156.R, B=@n156.R), B=@n156.B))
  n158 = Floor(None=Add(A=ComponentMask(None=@n157) {g=true, r=true}, B=@n146))
  n159 = ComponentMask(None=@n158) {r=true}
  n160 = Multiply(A=ComponentMask(None=@n154) {b=true, g=true}, B=@n159)
  n161 = ComponentMask(None=@n158) {g=true}
  n162 = Multiply(A=ComponentMask(None=@n154) {b=true, r=true}, B=@n161)
  n163 = ComponentMask(None=@n154) {g=true, r=true}
  n164 = OneMinus(None=Max(A=@n159, B=@n161))
  NamedRerouteDeclaration(None=Add(A=Add(A=@n160, B=@n162), B=Multiply(A=@n163, B=@n164))) {name=Noise_Local_UV, nodeColor=<rgba 1 0.7031 0 1>}
  n165 = NamedRerouteUsage()
  n166 = ScalarParameter() {defaultValue=1, group=Cover, parameterName=Border Dirt_SIDES_RoughnessMult, sortPriority=16}
  n167 = NamedRerouteUsage()
  n168 = ScalarParameter() {defaultValue=1, group=Cover, parameterName=Border Dirt_TOP_RoughnessMult, sortPriority=16}
  n169 = LinearInterpolate(
   A=Add(
    A=Multiply(A=@n165, B=@n166)
    B=ScalarParameter() {defaultValue=0.4, group=Cover, parameterName=Border Dirt_SIDES_RoughnessAdd, sortPriority=16}
   )
   B=Add(
    A=Multiply(A=@n167, B=@n168)
    B=ScalarParameter() {defaultValue=0.4, group=Cover, parameterName=Border Dirt_TOP_RoughnessAdd, sortPriority=16}
   )
   Alpha=NamedRerouteUsage()
  )
  n170 = Add(
   A=Multiply(A=NamedRerouteUsage(), B=ScalarParameter() {defaultValue=1, group=Cover, parameterName=Border Dirt_RoughnessMult, sortPriority=16})
   B=ScalarParameter() {defaultValue=0.4, group=Cover, parameterName=Border Dirt_RoughnessAdd, sortPriority=16}
  )
  n171 = NamedRerouteUsage()
  NamedRerouteDeclaration(None=Saturate(None=StaticSwitch(True=@n169, False=@n170, Value=@n171))) {name=BorderDirt_Roughness, nodeColor=<rgba 0.7578 1 0 1>}
  n172 = Constant2Vector() {g=0.01, r=0.01}
  NamedRerouteDeclaration(None=ShaderStageSwitch(PixelShader=@n114.DDX, VertexShader=@n172)) {name=DDX, nodeColor=<rgba 0 0.625 1 1>}
  NamedRerouteDeclaration(None=ShaderStageSwitch(PixelShader=@n114.DDY, VertexShader=@n172)) {name=DDY, nodeColor=<rgba 0.9453 1 0 1>}
  n173 = NamedRerouteUsage()
  n174 = StaticSwitchParameter(True=StaticBool() {value=true}, False=StaticBool()) {group=Snow, parameterName=UseSnowOcclusionTextureMask}
  n175 = StaticBool() {value=true}
  n176 = MaterialFunctionCall!(
   ScalarWeightMask=NamedRerouteUsage()
   UseSnowOcclusionTextureMask=@n174
   ------
   UseMaterialInput
   AdditionalMaskToRemoveSnow
   OutputTangentNormals=@n175
   UseAdditionalMaskToRemoveSnow
   UseSnowNormals
   OuputVertexNormalsAsTheSnowNormals
   UseSparkle
   LQSpark=@n175
   AllowSnowMicroNormals=StaticBool()
   CliffMaterialOnly
   Material=SetMaterialAttributes!(MaterialAttributes, Normal=NamedRerouteUsage()) {attributeSetTypes [#7]}
   CliffLandscapeLayerMask
   IsLandscape
   AreInputNormalsTangentSpace=@n175
   UseInputNormalsForFakeSnowShadow=StaticBool()
  ) {materialFunction=/Game/Athena/Helios/Environments/Landscape/Materials/QualityLevelLayers/MF_Helios_Snow_Default}
  NamedRerouteDeclaration(None=StaticSwitch(True=Multiply(A=@n173, B=@n176.SnowMask), False=@n176.SnowMask, Value=@n174)) {name=Landscape_H, nodeColor=<rgba 1 0.7266 0 1>}
  NamedRerouteDeclaration(None=@n16.OpacityMask) {name=OpacityMask_IN, nodeColor=<rgba 0.6875 0 1 1>}
  NamedRerouteDeclaration(None=@n83) {name=PreBlendCoverMask, nodeColor=<rgba 1 0 0.7734 1>}
  n177 = NamedRerouteUsage()
  n178 = ScalarParameter() {defaultValue=1, group=Cover, parameterName=Border Dirt_SIDES_SpecMult, sortPriority=16}
  n179 = NamedRerouteUsage()
  n180 = ScalarParameter() {defaultValue=1, group=Cover, parameterName=Border Dirt_TOP_SpecMult, sortPriority=16}
  n181 = LinearInterpolate(
   A=Add(A=Multiply(A=@n177, B=@n178), B=ScalarParameter() {defaultValue=0.4, group=Cover, parameterName=Border Dirt_SIDES_SpecAdd, sortPriority=16})
   B=Add(A=Multiply(A=@n179, B=@n180), B=ScalarParameter() {defaultValue=0.4, group=Cover, parameterName=Border Dirt_TOP_SpecAdd, sortPriority=16})
   Alpha=NamedRerouteUsage()
  )
  n182 = Add(
   A=Multiply(A=NamedRerouteUsage(), B=ScalarParameter() {defaultValue=1, group=Cover, parameterName=BoderDirt_SpecMult, sortPriority=16})
   B=ScalarParameter() {group=Cover, parameterName=BoderDirt_SpecAdd, sortPriority=16}
  )
  n183 = NamedRerouteUsage()
  NamedRerouteDeclaration(None=Saturate(None=StaticSwitch(True=@n181, False=@n182, Value=@n183))) {name=BorderDirt_Specular, nodeColor=<rgba 0.1719 1 0 1>}
  n184 = LinearInterpolate(
   A=ScalarParameter() {group=Cover, parameterName=Border Dirt_SIDES_FlattenNormals, sortPriority=17}
   B=ScalarParameter() {group=Cover, parameterName=Border Dirt_TOP_FlattenNormals, sortPriority=17}
   Alpha=NamedRerouteUsage()
  )
  n185 = ScalarParameter() {group=Cover, parameterName=Border Dirt_FlattenNormals, sortPriority=17}
  n186 = NamedRerouteUsage()
  NamedRerouteDeclaration(None=MaterialFunctionCall(Normal=NamedRerouteUsage(), Flatness=StaticSwitch(True=@n184, False=@n185, Value=@n186)) {materialFunction=&FlattenNormal}.Result) {name=BorderDirt_normal, nodeColor=<rgba 1 0 0.4922 1>}
  NamedRerouteDeclaration(None=FunctionInput() {bUsePreviewValueAsDefault=true, inputName=SFX V3, previewValue=<0 0 0 0>}) {name=IN_SFX, nodeColor=<rgba 0 0.3672 1 1>}
  n187 = StaticSwitch(
   True=VertexInterpolator(VS=LocalPosition() {localOrigin=InstancePreSkinning}.XY).PS
   False=WorldPosition() {worldPositionShaderOffset=WPT_ExcludeAllShaderOffsets}.XY
   Value=NamedRerouteUsage()
  )
  n188 = ScalarParameter() {defaultValue=1150, group=Cover, parameterName=Mask_WS Noise Tile, sortPriority=555}
  n189 = Divide(A=@n187, B=@n188) {constB=120}
  n190 = TextureObjectParameter() {group=Cover, parameterName=Cover WS Noise Texture, sortPriority=554, texture=/Game/Global/Textures/Default/Noise/T_Tiling_Noise_01}
  n191 = Divide(A=@n187, B=ScalarParameter() {defaultValue=900, group=Cover, parameterName=Mask_WS Noise Tile 2, sortPriority=555}) {constB=0.7}
  n192 = ScalarParameter() {defaultValue=1720, group=Cover, parameterName=Mask_WS Noise Tile 3, sortPriority=555}
  n193 = Add(
   A=TextureSample(UVs=Divide(A=@n187, B=@n192), Tex=@n190) {texture=&T_FlatGrey}.R
   B=ScalarParameter() {defaultValue=0.3, group=Cover, parameterName=Mask_WS Noise Tile 3_Bias, sortPriority=555}
  )
  n194 = ScalarParameter() {defaultValue=2, group=Cover, parameterName=Mask_WS Noise Tile 3_Contrast, sortPriority=555}
  n195 = NamedRerouteUsage()
  n196 = ScalarParameter() {defaultValue=10, desc=Mask_WS Noise Triplanar Fade, group=Cover, parameterName=Mask_WS Noise Triplanar Fade, sortPriority=555}
  n197 = MaterialFunctionCall(
   Texture Scale=NamedRerouteDeclaration(None=@n188) {name=Noise_Tile, nodeColor=<rgba 0 1 0.0625 1>}
   Triplanar Fade=NamedRerouteDeclaration(None=@n196) {name=Noise_triplanarFade, nodeColor=<rgba 0.3828 0 1 1>}
  ) {materialFunction=&MF_TriplanarDithered}
  n198 = StaticBoolParameter() {
   desc=Usefull for animated assets in case masks need to follow the skined mesh movement.
   group=Cover
   parameterName=Use Local Space UVs
   sortPriority=0
  }
  n199 = NamedRerouteDeclaration(None=@n198) {name=Use LS UV, nodeColor=<rgba 0.1953 0 1 1>}
  n200 = StaticSwitch(
   True=LinearInterpolate(
    A=TextureSample(UVs=@n189, Tex=@n190) {texture=&T_FlatGrey}.R
    B=TextureSample(UVs=@n191, Tex=@n190) {texture=&T_FlatGrey}.R
    Alpha=MaterialFunctionCall(In=@n193, Contrast=@n194) {materialFunction=&CheapContrast}.Result
   )
   False=TextureSample!(
    UVs=StaticSwitch(True=@n195, False=@n197.UVs, Value=@n199)
    Tex=@n190
    "DDX(UVs)"=NamedRerouteUsage()
    "DDY(UVs)"=NamedRerouteUsage()
    Apply View MipBias
   ) {mipValueMode=TMVM_Derivative, texture=&T_FlatGrey}.R
   Value=NamedRerouteDeclaration(None=StaticBoolParameter() {group=Cover, parameterName=Use Top Projection UVs, sortPriority=0}) {name=Use Top Proj, nodeColor=<rgba 0.5 1 0 1>}
  )
  n201 = ScalarParameter() {defaultValue=1, group=Cover, parameterName="Mask_WS Noise_Bias (Contract/Expand)", sortPriority=556}
  NamedRerouteDeclaration(None=Multiply(
    A=Add(A=@n200, B=@n201) {constB=-0.5}
    B=ScalarParameter() {defaultValue=1, group=Cover, parameterName="Mask_WS Noise_Mult (Hardness)", sortPriority=556}
   )) {name=M_WS noise, nodeColor=<rgba 0 0.5898 1 1>}
  n202 = Constant() {r=0.5}
  NamedRerouteDeclaration(None=ShadingPathSwitch(
    Default=@n176.SnowMaterialOnly
    Mobile=SetMaterialAttributes!(MaterialAttributes, BaseColor=@n176.SnowAlbedo, Specular=@n202, Roughness=@n202) {attributeSetTypes [#6, #4, #3]}
   )) {name=Landscape_Snow, nodeColor=<rgba 0 1 0.1094 1>}
  NamedRerouteDeclaration(None=@n16.Roughness) {name=Roughness_IN, nodeColor=<rgba 0.8984 0 1 1>}
  Saturate(None=@n70)
  SetMaterialAttributes!(MaterialAttributes, EmissiveColor=ComponentMask(None=NamedRerouteUsage()) {r=true}) {attributeSetTypes [#1]}
  Transform(None=NamedRerouteUsage())
 ]
}
